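; Creative matchmaking portal device that lets players start public or private matchmaking into the island set by its code.
; Loaded link data triggers a moderation check that drives prompts, visuals and either an on-portal widget or in-world HUD.
; Overlap and matchmaking events toggle interaction UI; interacting can open an island details screen.
; The server manages code locking, restores the knob code, and periodically re-sets the current code to catch republished islands.
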(event EventBeginPlay
  (|CallParentFunction)
  (CallFunction|UpdateThumbnailDetails false)
  (if (Matchmaking|IsonPortalHudEnabled)
    (bind _returnvalue (UserInterface|CreateWidget "/Game/Creative/Devices/MatchmakingPortal/MatchmakingOnPortalDetails.MatchmakingOnPortalDetails_C"))
    (Variables|Default|SetOnPortalDetails _returnvalue)
    (if (CallFunction|ShouldHavePortalUIOverlay)
      (CallFunction|InitializePortalUIOverlay)
      (else
        _))
    (Default|UpdateInWorldHUDVisibility)
    (else
      _)))

(event EventOnMinigameStarted
  (|CallParentFunction)
  (Variables|Interaction|SetbShowSecondInteractPrompt (Interaction|ShouldDisplaySecondInteraction))
  (Variables|Interaction|SetbShowFirstInteractPrompt (Variables|Default|GetHasValidModerationState))
  (CallFunction|UpdatePortalInteractionUIUsingLastInteracter)
  (Variables|Default|SetPrivatePrimaryMatchmaking (or (== (Variables|UserOptions|GetJoinOption) 2) (not (Variables|Default|GetHasSecondInteraction))))
  (Networking|SwitchHasAuthority
    (:Authority
      (CodeOverride|UpdateLockMode))))

(event EventOnMinigameEnded
  (|CallParentFunction)
  (Variables|Interaction|SetbShowSecondInteractPrompt false)
  (Variables|Interaction|SetbShowFirstInteractPrompt true)
  (CallFunction|DisablePortalInteractionUI)
  (Networking|SwitchHasAuthority
    (:Authority
      (CallFunction|RestoreIslandCodeUsingKnobCode))))

(event OnLinkDataLoaded(FortPortal)
  (params (LinkData :type Fort%20Game%20Activity%20Object%20Reference))
  (bind _returnvalue (CallFunction|LinkIdtoString (Class|FortGameActivity|GetQueryLinkId LinkData)))
  (Variables|Default|SetCurrent%20Island%20Code _returnvalue)
  (Networking|SwitchHasAuthority
    (:Authority
      (Utilities|FlowControl|Sequence
        (:then_0
          ;[1 1 1] If the Island Code was changed, we need to update the lock mode
          (if (and (CallFunction|CodeIslandWasChanged) (not (Utilities|String|IsEmpty (Variables|Default|GetCurrent%20Island%20Code))))
            (CodeOverride|UpdateLockMode)))
        (:then_1
          (Variables|Default|SetHasSecondInteraction (not (or (Utilities|String|EqualExactly(String) (Utilities|Struct|BreakCreativeIslandMatchmakingSettings (Class|FortGameActivity|GetMatchmakingSettingsV1 LinkData)) "off") (Utilities|String|EqualExactly(String) (Utilities|Struct|BreakCreativeIslandMatchmakingSettings (Class|FortGameActivity|GetMatchmakingSettingsV1 LinkData)) "off_override")))))
        (:then_2
          ;[1 0 0] Start timer to set current code again to handle the case when the pointed island is re-published
          (if (Default|HasValidCode)
            (CallFunction|StartTimerRetrySetCurrentCode)))))))

(event OnAddedtoMinigame(FortMinigameProgress)
  (params (Minigame :type Fort%20Minigame%20Object%20Reference))
  (Networking|SwitchHasAuthority
    (:Authority
      (Variables|Default|SetCachedMinigame Minigame)
      (Minigame|AssignPlayerAddedDelegate Minigame (EventDispatchers|CreateEvent :!event On%20Player%20Added))
      (Minigame|BindEventtoPlayerRemovedDelegate Minigame (EventDispatchers|CreateEvent :!event On%20Player%20Removed)))))

(event OnRemovedfromMinigame(FortMinigameProgress)
  (params (Minigame :type Fort%20Minigame%20Object%20Reference))
  (Networking|SwitchHasAuthority
    (:Authority
      (Utilities|Time|ClearandInvalidateTimerbyHandle (Variables|Default|GetRetrySetCodeTimerHandle))
      (Minigame|UnbindEventfromPlayerAddedDelegate Minigame (EventDispatchers|CreateEvent :!event On%20Player%20Added))
      (Minigame|UnbindEventfromPlayerRemovedDelegate Minigame (EventDispatchers|CreateEvent :!event On%20Player%20Removed)))))

(event OnPortalLinkLockExpired(FortPortal)
  (params (Status :type EPortalLinkCodeLockStatus%20Enum :default "Unlocked_NotSet"))
  (Networking|SwitchHasAuthority
    (:Authority
      (if (and (CallFunction|ResetCodeAfterCooldown) (Minigame|IsinProgress (Variables|Default|GetCachedMinigame)))
        (CallFunction|RestoreIslandCodeUsingKnobCode)))))

(event OnComponentBeginOverlap(HUDVisibilityVolume)
  (params (OverlappedComponent :type Primitive%20Component%20Object%20Reference) (OtherActor :type Actor%20Object%20Reference) (OtherComp :type Primitive%20Component%20Object%20Reference) (OtherBodyIndex :type Integer) (bFromSweep :type Boolean) (SweepResult :type Hit%20Result%20Structure%20%28by%20ref%29))
  (if (not (Math|Boolean|ORBoolean (Variables|Default|GetMatchmakingOnCooldown) (Variables|Default|GetIsMatchmaking) (not (Class|FortPortalComponent|HasValidLinkData (Variables|Default|GetFortPortal)))))
    (bind _asfort_pawn (Utilities|Casting|CastToFortPawn OtherActor))
    (CallFunction|UpdatePortalInteractionUI _asfort_pawn)))

(event OnComponentEndOverlap(HUDVisibilityVolume)
  (params (OverlappedComponent :type Primitive%20Component%20Object%20Reference) (OtherActor :type Actor%20Object%20Reference) (OtherComp :type Primitive%20Component%20Object%20Reference) (OtherBodyIndex :type Integer))
  (if (and (HelperFunctions|IsLocalPawn OtherActor) (Class|FortPortalComponent|HasValidLinkData (Variables|Default|GetFortPortal)))
    (CallFunction|DisablePortalInteractionUI)
    (if (Utilities|Equal(Object) OtherActor (Variables|Default|GetLastInteractionPawn))
      (Variables|Default|SetLastInteractionPawn 0))))

(event EventEndPlay
  (params (EndPlayReason :type EEndPlayReason%20Enum :default "Destroyed"))
  (CallFunction|ReleaseKeybinds))

(event ChildFunctions|EventPlayModeChanged
  (params (Minigame :type Fort%20Minigame%20Object%20Reference) (bIsInPlayMode :type Boolean))
  (|CallParentFunction Minigame bIsInPlayMode)
  (CallFunction|UpdatePortalVisuals))

(fn AthenaInteractDetails
  (return (Variables|CodeOverride|GetLocalPlayerLookingAtScreen) (CallFunction|IsinPlayMode)))

(fn BlueprintCanInteract
  (params (InteractingPawn :type Fort%20Pawn%20Object%20Reference) (InteractionBeingAttempted :type EInteractionBeingAttempted%20Enum :default "FirstInteraction") (InteractionType :type TInteractionType%20Enum :default "IT_NoInteraction"))
  (if (Utilities|Enum|Equal(Enum) InteractionBeingAttempted "SecondInteraction")
    (if (and (Input|IsUsingGamepad (Fortnite|GetFortPlayerControllerfromActor (Variables|GetInteractingPawn))) (Class|HeaderDescriptionHUDComponent|IsNewMatchmakingPortalUIEnabled))
      (return false)
      (else
        _)))
  (return (|CallParentFunction (Variables|GetInteractingPawn) (Variables|GetInteractionBeingAttempted) (Variables|GetInteractionType))))

(fn BlueprintGetInteractionString
  (params (InteractingPawn :type Fort%20Pawn%20Object%20Reference) (InteractionBeingAttempted :type EInteractionBeingAttempted%20Enum :default "FirstInteraction"))
  (bind _interactingpawn (Variables|GetInteractingPawn))
  (bind _returnvalue (Fortnite|GetFortPlayerControllerfromActor _interactingpawn))
  (bind _client_portal_components (Variables|Default|GetClient%20Portal%20Components))
  (bind (_value _returnvalue_2) (Utilities|Map|Find _client_portal_components _returnvalue))
  (Networking|CanExecuteCosmeticEvents
    (:True
      (switch Utilities|FlowControl|Switch|SwitchonEInteractionBeingAttempted InteractionBeingAttempted
        (:FirstInteraction
          (if (not (Math|Boolean|ORBoolean (Interaction|CanEditOptionsComponent (Variables|Default|GetToyOptionsComponent) (Variables|GetInteractingPawn)) (Variables|CodeOverride|GetLocalPlayerLookingAtScreen) (not (Matchmaking|CanInitiatePrivateMatchmaking))))
            (Utilities|IsValid _returnvalue
              (:Is%20Valid
                (if _returnvalue_2
                  (if (CallFunction|IsPartyLeaderandHasActiveReadyCheck _returnvalue)
                    _
                    (elif (Class|FortControllerComponentPortal|CanInitiatePortalMatchmkaing _value (Variables|Default|GetFortPortal))
                      (return (select (or (not (Variables|Default|GetHasSecondInteraction)) (== (Variables|UserOptions|GetJoinOption) 2)) (Variables|Default|GetPrivateText) (Variables|Default|GetPublicText)))))
                  (bind _returnvalue_3 (|CallParentFunction (Variables|GetInteractingPawn)))
                  (return _returnvalue_3)
                  (else
                    _)))
              (:Is%20Not%20Valid
                _))
            (return "")))
        (:SecondInteraction
          (return (Variables|Default|GetPrivateText)))
        (:AllInteraction
          _)))
    (:False
      _))
  (return ""))

(fn BlueprintGetInteractionTime
  (params (InteractingPawn :type Fort%20Pawn%20Object%20Reference) (InteractionBeingAttempted :type EInteractionBeingAttempted%20Enum :default "FirstInteraction"))
  (if (Matchmaking|IsGameDetailPreviewScreenEnabled)
    (return 0.0 true)
    (else
      (bind (_outinteractiontime _returnvalue) (|CallParentFunction InteractingPawn InteractionBeingAttempted))
      (return _outinteractiontime _returnvalue))))

(fn BlueprintOnLocalInteract
  (params (InteractingPawn :type Fort%20Player%20Pawn%20Object%20Reference) (InteractionBeingAttempted :type EInteractionBeingAttempted%20Enum :default "FirstInteraction"))
  (bind _interactingpawn (Variables|GetInteractingPawn))
  (switch Utilities|FlowControl|Switch|SwitchonEInteractionBeingAttempted InteractionBeingAttempted
    (:FirstInteraction
      (Variables|Default|SetbPrivateMatchmaking (Variables|Default|GetPrivatePrimaryMatchmaking)))
    (:SecondInteraction
      (if (and (CallFunction|IsinPlayMode) (not (Variables|CodeOverride|GetLocalPlayerLookingAtScreen)))
        (Variables|Default|SetbPrivateMatchmaking true)
        (else
          (return false)))))
  (if (and (CallFunction|IsinPlayMode) (not (Variables|CodeOverride|GetLocalPlayerLookingAtScreen)))
    (if (Variables|Default|GetHasValidModerationState)
      (if (Matchmaking|IsGameDetailPreviewScreenEnabled)
        (CallFunction|SpawnIslandDetailsUI _interactingpawn)
        (return true)
        (else
          _))
      (else
        (return false)))
    (else
      _))
  (bind _consumesinput (Interaction|HandleInteraction _interactingpawn))
  (return _consumesinput))

(fn BP_ShouldSuppressInteractionWidget
  (return (Class|HeaderDescriptionHUDComponent|IsNewMatchmakingPortalUIEnabled)))

(fn CanMatchmake
  (params (InteractingPlayer :type Fort%20Player%20Pawn%20Object%20Reference))
  (bind _returnvalue (Fortnite|GetFortPlayerControllerfromActor InteractingPlayer))
  (bind _client_portal_components (Variables|Default|GetClient%20Portal%20Components))
  (bind (_value _returnvalue_2) (Utilities|Map|Find _client_portal_components _returnvalue))
  (Utilities|IsValid InteractingPlayer
    (:Is%20Valid
      (if _returnvalue_2
        (return (Class|FortControllerComponentPortal|CanInitiatePortalMatchmkaing _value (Variables|Default|GetFortPortal)))
        (else
          _)))
    (:Is%20Not%20Valid
      _))
  (return false))

(fn CanInteract
  (return (and (not (UserOptions|HasEnableasArtPromptKnobOptionActivated)) (|CallParentFunction))))

(fn CanPlayPortalAudio
  (return (and (|CallParentFunction) (UserOptions|HasEnableAudioKnobOptionActivated))))

(fn CodeIslandWasChanged
  (return (not (Utilities|String|Contains (Variables|UserOptions|GetIslandCode_Knob) (Variables|Default|GetCurrent%20Island%20Code)))))

(fn ConsoleIsVisible
  (return (|CallParentFunction)))

(fn DisableAndWaitForPortalInteractionUIUpdate
  (params (WaitTime :type Float%20%28double-precision%29 :default 1.0))
  (CallFunction|DisablePortalInteractionUI)
  (Utilities|Time|SetTimerbyEvent (EventDispatchers|CreateEvent :!event Update%20Portal%20Interaction%20if%20Not%20on%20Cooldown) WaitTime))

(fn DisablePortalInteractionUI
  (Networking|CanExecuteCosmeticEvents
    (:True
      (bind _onportaldetails (Variables|Default|GetOnPortalDetails)
        (:then
          (Class|MatchmakingonPortalDetails|DisableInteractions _onportaldetails)
          (UserInterface|RequestRenderUpdate (Variables|Default|GetPortalWidgetComponent)))
        (:else)))))

(fn GetCollisionStaticMesh
  (return "0, 0, 0" (Variables|/Script/Engine.StaticMeshComponent|GetStaticMesh (Variables|Default|GetCode%20Entry%20Screen))))

(fn GetInteractionText
  (params (InteractingPawn :type Fort%20Pawn%20Object%20Reference) (InteractionBeingAttempted :type EInteractionBeingAttempted%20Enum :default "FirstInteraction"))
  (bind _interactingpawn (Variables|GetInteractingPawn))
  (bind _returnvalue (Fortnite|GetFortPlayerControllerfromActor _interactingpawn))
  (bind _fortportal (Variables|Default|GetFortPortal))
  (bind _returnvalue_2 (Class|FortControllerComponentPortal|CanInitiatePortalMatchmkaing (Utilities|Map|Find (Variables|Default|GetClient%20Portal%20Components) _returnvalue) _fortportal))
  (bind _client_portal_components (Variables|Default|GetClient%20Portal%20Components))
  (bind (_value _returnvalue_3) (Utilities|Map|Find _client_portal_components _returnvalue))
  (bind _interactionbeingattempted (Variables|GetInteractionBeingAttempted))
  (if (and (Utilities|Enum|NotEqual(Enum) InteractionBeingAttempted "AllInteraction") (CallFunction|IsinPlayMode))
    (if (and (Utilities|IsValid _returnvalue :!class /Script/Engine.KismetSystemLibrary) _returnvalue_3)
      (if (CallFunction|IsPartyLeaderandHasActiveReadyCheck _returnvalue)
        (return (Variables|Default|GetStop%20Matchmaking%20Text) (Utilities|Enum|Equal(Enum) _interactionbeingattempted "FirstInteraction"))
        (else
          (switch Utilities|FlowControl|Switch|SwitchonEInteractionBeingAttempted _interactionbeingattempted
            (:FirstInteraction
              (if _returnvalue_2
                (return (CallFunction|GetPrimaryPortalInteractionText) true)
                (elif (Class|FortControllerComponentPortal|CanCancelPortalMatchmaking _value)
                  (return (Variables|Default|GetStop%20Matchmaking%20Text) true)
                  (elif (Class|FortPortalComponent|IsPartyEligibleforMatchmaking _fortportal)
                    (return "" false)
                    (else
                      (return (Variables|Default|GetParty%20Ineligible%20Text) true))))))
            (:SecondInteraction
              (return (Variables|Default|GetPrivateText) (Math|Boolean|ANDBoolean _returnvalue_2 (Variables|Default|GetHasSecondInteraction) (Matchmaking|CanInitiatePrivateMatchmaking) (not (Variables|Default|GetIsMatchmaking)) (== (Variables|UserOptions|GetJoinOption) 0)))))))
      (else
        (return "" false)))
    (else
      (return "" false))))

(fn GetInWorldHUD
  (bind _inworldhudwidget (Variables|Default|GetInWorldHUDWidget)
    (:then
      (bind _asmatchmaking_in_world_details (Utilities|Casting|CastToMatchmakingInWorldDetails (UserInterface|GetUserWidgetObject _inworldhudwidget))
        (:then
          (return _asmatchmaking_in_world_details))
        (:CastFailed
          (return 0))))
    (:else)))

(fn GetMeshComponents
  (return 0))

(fn GetPrimaryPortalInteractionText
  (return (select (or (not (Variables|Default|GetHasSecondInteraction)) (== (Variables|UserOptions|GetJoinOption) 2)) (Variables|Default|GetPrivateText) (Variables|Default|GetPublicText))))

(fn GetVisibilityComponents
  (return (Utilities|Array|MakeArray (Variables|Default|GetInWorldHUDWidget))))

(fn HandleInteraction
  (params (InteractingPawn :type Fort%20Player%20Pawn%20Object%20Reference))
  (bind _outputdelegate (EventDispatchers|CreateEvent :!event OnReadyCheckCancelled))
  (bind _returnvalue (|CallParentFunction InteractingPawn "FirstInteraction"))
  (Utilities|FlowControl|Sequence
    (:then_0
      (bind _readycheckinteractionplayer (Variables|Default|GetReadyCheckInteractionPlayer)
        (:then
          (EventDispatchers|UnbindEventfromInGameMatchmakingReadyCheckCanceled _readycheckinteractionplayer _outputdelegate)
          (EventDispatchers|BindEventtoInGameMatchmakingReadyCheckCanceled _readycheckinteractionplayer _outputdelegate))
        (:else)))
    (:then_1
      (if (HelperFunctions|IsLocalPawn (Variables|GetInteractingPawn))
        (Variables|Default|SetLastInteractionPawn (Variables|GetInteractingPawn))
        (else
          _))
      (return _returnvalue))))

(fn HandleMatchmakingThrottleTimerComplete
  (|CallParentFunction)
  (CallFunction|UpdatePortalInteractionIfNotOnCooldown))

(fn HasSecondInteraction
  (params (InteractingPawn :type Fort%20Pawn%20Object%20Reference))
  (bind _interactingpawn (Variables|GetInteractingPawn))
  (bind _returnvalue (Fortnite|GetFortPlayerControllerfromActor _interactingpawn))
  (bind _client_portal_components (Variables|Default|GetClient%20Portal%20Components))
  (bind (_value _returnvalue_2) (Utilities|Map|Find _client_portal_components _returnvalue))
  (Networking|CanExecuteCosmeticEvents
    (:True
      (if (or (Interaction|CanEditOptionsComponent (Variables|Default|GetToyOptionsComponent) (Variables|GetInteractingPawn)) (Variables|CodeOverride|GetLocalPlayerLookingAtScreen))
        (return false)
        (else
          (Utilities|IsValid _returnvalue
            (:Is%20Valid
              (if _returnvalue_2
                (if (CallFunction|IsPartyLeaderandHasActiveReadyCheck _returnvalue)
                  (return false)
                  (elif (Class|FortControllerComponentPortal|CanInitiatePortalMatchmkaing _value (Variables|Default|GetFortPortal))
                    (return (and (Variables|Default|GetHasSecondInteraction) (== (Variables|UserOptions|GetJoinOption) 0)))
                    (else
                      (return false))))
                (else
                  _)))
            (:Is%20Not%20Valid
              _))
          (return false))))
    (:False
      (return false))))

(fn HasEnableAsArtPromptKnobOptionActivated
  (return (Math|Conversions|ToBoolean(Integer) (Variables|UserOptions|GetKnobEnableAsArtPrompt))))

(fn HasEnableAudioKnobOptionActivated
  (return (Math|Conversions|ToBoolean(Integer) (Variables|UserOptions|GetKnobEnableAudio))))

(fn HasValidCode
  (return (and (>= (Utilities|String|Len (Variables|Default|GetCurrent%20Island%20Code)) (Variables|Default|GetValid%20Code%20Length)) (Utilities|String|NotEqualExactly(String) (Variables|Default|GetCurrent%20Island%20Code) (Variables|Default|GetDefault%20Code)))))

(fn IgnoreWorldGridScaleOverride
  (return false))

(fn InitializeDevice
  (|CallParentFunction)
  (if (Class|FortPortalComponent|HasValidLinkData (Variables|Default|GetFortPortal))
    (Default|UpdateInWorldHUDVisibility)
    (if (CallFunction|ShouldHavePortalUIOverlay)
      (CallFunction|InitializePortalUIOverlay)
      (bind _onportaldetails (Variables|Default|GetOnPortalDetails)
        (:then
          (Class|MatchmakingonPortalDetails|UpdatePortalUI _onportaldetails)
          (Class|MatchmakingonPortalDetails|SetIslandTitleVisibility (Variables|Default|GetOnPortalDetails) (UserOptions|IsPortalNameVisible))
          (Class|MatchmakingonPortalDetails|SetIslandDetailsVisibility (Variables|Default|GetOnPortalDetails) (UserOptions|IsFloatingHUDVisible))
          (UserInterface|RequestRenderUpdate (Variables|Default|GetPortalWidgetComponent)))
        (:else
          _))
      (Default|ScaleandPositionNoThumbnailPlane)
      (for _array_element (Variables|CodeOverride|GetLocal%20Pawns%20in%20Range)
        (CallFunction|SetHUDVisibility _array_element true))
      (else
        (Default|RemovePortalUIOverlay)))
    (else
      _)))

(fn InitializePortalUIOverlay
  (Networking|CanExecuteCosmeticEvents
    (:True
      (if (Matchmaking|IsonPortalHudEnabled)
        (Variables|Default|GetPortalWidgetComponent
          (:then
            _)
          (:else
            (bind _returnvalue (Game|AddComponentbyClass 0 "/Game/Creative/Devices/MatchmakingPortal/Creative_MatchmakingPortalWidgetComponent.Creative_MatchmakingPortalWidgetComponent_C" false (Variables|Default|GetWidgetComponentTransform)))
            (bind _output_get (Variables|Default|SetPortalWidgetComponent _returnvalue))
            (Transformation|AttachComponentToComponent _output_get (Variables|Default|GetCapsule) "None" "KeepRelative" "KeepRelative" "KeepRelative" false)))
        (bind _onportaldetails (Variables|Default|GetOnPortalDetails)
          (:then
            (UserInterface|SetWidget (Variables|Default|GetPortalWidgetComponent) _onportaldetails)
            (Variables|/Game/Creative/Devices/MatchmakingPortal/MatchmakingOnPortalDetails.MatchmakingOnPortalDetails_C|SetMatchmakingPortal self _onportaldetails)
            (Class|MatchmakingonPortalDetails|UpdatePortalUI _onportaldetails)
            (UserInterface|RequestRenderUpdate (Variables|Default|GetPortalWidgetComponent)))
          (:else
            (bind _asmatchmaking_on_portal_details (Utilities|Casting|CastToMatchmakingOnPortalDetails (UserInterface|GetUserWidgetObject (Variables|Default|GetPortalWidgetComponent))))
            (bind _output_get_2 (Variables|Default|SetOnPortalDetails _asmatchmaking_on_portal_details))
            (Variables|/Game/Creative/Devices/MatchmakingPortal/MatchmakingOnPortalDetails.MatchmakingOnPortalDetails_C|SetMatchmakingPortal self _output_get_2)))
        (else
          (Variables|Default|SetOnPortalDetails 0)
          (Variables|Default|SetPortalWidgetComponent 0))))))

(fn IsHUDVisible
  (return (Math|Boolean|ANDBoolean (Class|CreativeEnabledComponent|IsEnabled (Variables|Default|GetEnabledComponent)) (Default|HasValidCode) (Class|FortPortalComponent|HasValidLinkData (Variables|Default|GetFortPortal)))))

(fn IsFloatingHUDVisible
  (return (and (!= (Variables|UserOptions|GetShowFloatingHUD) 0) (Variables|Default|GetHasValidModerationState))))

(fn IslandDataLoaded
  (params (LinkData :type Fort%20Game%20Activity%20Object%20Reference%20%28by%20ref%29))
  (bind _output_get (Variables|Default|SetIslandLinkData LinkData))
  (Class|FortPortalComponent|CheckIslandModeration (Variables|Default|GetFortPortal) _output_get (EventDispatchers|CreateEvent :!event OnModerationCheckComplete))
  (|CallParentFunction _output_get)
  (if (Matchmaking|IsonPortalHudEnabled)
    (Networking|CanExecuteCosmeticEvents
      (:True
        (if (CallFunction|ShouldHavePortalUIOverlay)
          (CallFunction|InitializePortalUIOverlay)
          (bind _onportaldetails (Variables|Default|GetOnPortalDetails)
            (:then
              (Class|MatchmakingonPortalDetails|UpdatePortalUI _onportaldetails)
              (CallFunction|UpdatePortalInteractionUIUsingLastInteracter)
              (Default|UpdateInWorldHUDVisibility))
            (:else
              _))
          (return)
          (else
            (Default|RemovePortalUIOverlay))))
      (:False
        _))
    (else
      (Default|UpdateInWorldHUDVisibility))))

(fn IslandDataNoSuccess
  (params (Matchmaking%20State :type ECreativeMatchmakingPortalState%20Enum :default "NewEnumerator0"))
  (|CallParentFunction Matchmaking%20State)
  (Default|UpdateInWorldHUDVisibility))

(fn IsMatchmakingEnabledTextVisible
  (return (Math|Boolean|ANDBoolean (!= (Variables|UserOptions|GetShowMatchmakingEnabledText) 0) (UserOptions|IsHUDVisible) (Variables|Default|GetHasValidModerationState))))

(fn IsPortalNameVisible
  (return (and (!= (Variables|UserOptions|GetShowPortalName) 0) (UserOptions|IsHUDVisible))))

(fn MakeSafeOnlineLinkCode
  (params (Link%20Code :type String))
  (if (UserOptions|HasEnableasArtPromptKnobOptionActivated)
    (bind _outlinkid (Class|CreativeBlueprintLibrary|MakeOnlineLinkCode Link%20Code))
    (return _outlinkid)
    (else
      (bind _link_id (|CallParentFunction Link%20Code))
      (return _link_id))))

(fn MatchmakingPortalHasSecondInteraction
  (params (InteractingPawn :type Fort%20Player%20Pawn%20Object%20Reference))
  (Utilities|IsValid InteractingPawn
    (:Is%20Valid
      (bind _return (Interaction|HasSecondInteraction InteractingPawn))
      (return _return))
    (:Is%20Not%20Valid
      (return false))))

(fn NeedsKeybindSetup
  (return (and (Utilities|IsValid (Variables|Default|GetOnPortalDetails) :!class /Script/Engine.KismetSystemLibrary) (Math|Boolean|NORBoolean (Utilities|IsValid (Variables|Default|GetUseWidget) :!class /Script/Engine.KismetSystemLibrary) (Utilities|IsValid (Variables|Default|GetAltWidget) :!class /Script/Engine.KismetSystemLibrary)))))

(fn OnMatchmakingComplete
  (params (bSuccess :type Boolean))
  (|CallParentFunction bSuccess)
  (if (not bSuccess)
    (Variables|Default|SetIsMatchmaking false)
    (Variables|Interaction|SetbShowSecondInteractPrompt (Interaction|ShouldDisplaySecondInteraction))
    (CallFunction|DisableandWaitforPortalInteractionUIUpdate 1.0)))

(fn OnPlayerAdded
  (params (UniqueNetId :type Unique%20Net%20Id%20Structure) (bIsLocalPlayer :type Boolean))
  (if (and (not (Utilities|Time|IsTimerActivebyHandle (Variables|Default|GetRetrySetCodeTimerHandle))) (not (Utilities|String|IsEmpty (Variables|UserOptions|GetIslandCode_Knob))))
    (CallFunction|StartTimerRetrySetCurrentCode)
    (return)))

(fn OnPlayerRemoved
  (params (UniqueNetId :type Unique%20Net%20Id%20Structure) (bIsLocalPlayer :type Boolean))
  (if (not (Utilities|IsArrayNonEmpty (Minigame|GetParticipatingPlayers (Variables|Default|GetCachedMinigame))))
    (Utilities|Time|ClearandInvalidateTimerbyHandle (Variables|Default|GetRetrySetCodeTimerHandle))))

(fn OnModerationCheckComplete
  (params (bValidIsland :type Boolean))
  (bind _output_get (Variables|Default|SetHasValidModerationState bValidIsland))
  (Networking|CanExecuteCosmeticEvents
    (:True
      (Variables|Default|SetbThumbnailPlaneIsHidden (not _output_get))
      (CallFunction|UpdatePortalVisuals)
      (if (Matchmaking|IsonPortalHudEnabled)
        (if (CallFunction|ShouldHavePortalUIOverlay)
          (CallFunction|InitializePortalUIOverlay)
          (bind _onportaldetails (Variables|Default|GetOnPortalDetails)
            (:then
              (Class|MatchmakingonPortalDetails|UpdatePortalUI _onportaldetails)
              (UserInterface|RequestRenderUpdate (Variables|Default|GetPortalWidgetComponent))
              (CallFunction|UpdatePortalInteractionIfNotOnCooldown))
            (:else))
          (else
            (Default|RemovePortalUIOverlay)))
        (else
          (CallFunction|UpdateHUD (Variables|IslandDetails|GetCurrent%20Matchmaking%20State)))))))

(fn OnReadyCheckCancelled
  (Variables|Default|SetIsMatchmaking false)
  (Variables|Interaction|SetbShowSecondInteractPrompt (Interaction|ShouldDisplaySecondInteraction))
  (CallFunction|DisableandWaitforPortalInteractionUIUpdate 1.0))

(fn OnRep_HasSecondInteraction
  (if (not (Variables|Default|GetHasSecondInteraction))
    (Variables|Default|SetPrivatePrimaryMatchmaking true)))

(fn PortalHUDIsVisible
  (params (bVisibility :type Boolean))
  (return (Math|Boolean|ANDBoolean (UserOptions|IsFloatingHUDVisible) (|CallParentFunction bVisibility) (not (Matchmaking|IsonPortalHudEnabled)))))

(fn PortalThumbnailIsVisible
  (return (and (|CallParentFunction) (not (Variables|Default|GetbThumbnailPlaneIsHidden)))))

(fn ReleaseKeybinds
  (Default|CallKeybindRelease self)
  (Variables|Default|SetUseWidget 0)
  (Variables|Default|SetAltWidget 0))

(fn RemovePortalUIOverlay
  (bind _portalwidgetcomponent (Variables|Default|GetPortalWidgetComponent)
    (:then
      (Components|DestroyComponent _portalwidgetcomponent)
      (Variables|Default|SetPortalWidgetComponent 0))
    (:else)))

(fn ResetCodeAfterCooldown
  (return (== (Variables|Default|GetCodeStatusAfterCooldown) 0)))

(fn RestoreIslandCodeUsingKnobCode
  (if (CallFunction|CodeIslandWasChanged)
    (Variables|/Script/FortniteGame.FortPortalComponent|SetLinkCodeLockMode "NeverLocked" (Variables|Default|GetFortPortal))
    (bind _returnvalue (CallFunction|CreateLinkId (Variables|UserOptions|GetIslandCode_Knob)))
    (Class|FortPortalComponent|ServerSetLinkCode (Variables|Default|GetFortPortal) _returnvalue (Variables|DAD|GetbDADRequestedUpdate) false false)))

(fn RetrySetCurrentCode
  (if (Default|HasValidCode)
    (bind _returnvalue (CallFunction|CreateLinkId (Variables|Default|GetCurrent%20Island%20Code)))
    (Class|FortPortalComponent|ServerSetLinkCode (Variables|Default|GetFortPortal) _returnvalue true false false)
    (else
      (Utilities|Time|ClearandInvalidateTimerbyHandle (Variables|Default|GetRetrySetCodeTimerHandle)))))

(fn ScaleAndPositionNoThumbnailPlane
  (Transformation|SetRelativeLocation (Variables|Default|GetNoThumbnailPlane) (Variables|Default|GetNoThumbnailBasePosition) false false))

(fn SetUpMatchmakingHUDWidget
  (params (InteractingPawn :type Fort%20Pawn%20Object%20Reference))
  (bind _returnvalue (Fortnite|GetFortPlayerControllerfromActor InteractingPawn))
  (Networking|CanExecuteCosmeticEvents
    (:True
      (if (Matchmaking|IsonPortalHudEnabled)
        (Variables|Default|GetMatchmakingHUD
          (:then
            _)
          (:else
            (bind _returnvalue_2 (UserInterface|CreateWidget "/Game/Creative/Devices/MatchmakingPortal/MatchmakingPortalHUD.MatchmakingPortalHUD_C" _returnvalue))
            (bind _output_get (Variables|Default|SetMatchmakingHUD _returnvalue_2))))
        (Class|MatchmakingPortalHUD|SetupIslandInfo _output_get (Variables|Default|GetIslandLinkData))
        (bind _asfort_player_controller_gameplay (Utilities|Casting|CastToFortPlayerControllerGameplay _returnvalue))
        (HUD|SetHUDElementVisibilitywithContext _asfort_player_controller_gameplay (GameplayTags|MakeLiteralGameplayTagContainer "(GameplayTags=((TagName=\"HUD.TopLeft.TeamPlayers\"),(TagName=\"Athena.Creative.MatchmakingFeedback.Standard\")))") true "MatchmakingPortal" "")))))

(fn SetUpMatchmakingInteractions
  (Variables|Default|SetIsMatchmaking true)
  (Variables|Interaction|SetbShowSecondInteractPrompt false)
  (CallFunction|DisablePortalInteractionUI)
  (CallFunction|SetUpMatchmakingHUDWidget (Variables|Default|GetLastInteractionPawn)))

(fn SetupKeybinds
  (params (UseWidget :type Keybind%20Widget%20Object%20Reference) (AltWidget :type Keybind%20Widget%20Object%20Reference))
  (bind _usewidget (Variables|Default|GetUseWidget))
  (bind _altwidget (Variables|Default|GetAltWidget))
  (Utilities|FlowControl|Sequence
    (:then_0
      (if (or (Utilities|IsValid (Variables|Default|GetUseWidget) :!class /Script/Engine.KismetSystemLibrary) (Utilities|IsValid (Variables|Default|GetAltWidget) :!class /Script/Engine.KismetSystemLibrary))
        (CallFunction|ReleaseKeybinds)))
    (:then_1
      (bind _onportaldetails (Variables|Default|GetOnPortalDetails)
        (:then
          (Class|MatchmakingonPortalDetails|SetupKeybinds _onportaldetails _usewidget _altwidget)
          (Variables|Default|SetUseWidget _usewidget)
          (Variables|Default|SetAltWidget _altwidget))
        (:else)))))

(fn SetupMeshTargetToCheckVisibility
  (Variables|VisibilityCheck|SetMeshTargetToCheckHUDVisibility (Variables|Default|GetThumbnailPlane)))

(fn SetupWaitForReadyCheck
  (params (FortPlayer :type Fort%20Player%20Controller%20Athena%20Object%20Reference))
  (|CallParentFunction FortPlayer)
  (Default|SetUpMatchmakingInteractions))

(fn ShouldHavePortalUIOverlay
  (return (Math|Boolean|NANDBoolean (UserOptions|HasEnableasArtPromptKnobOptionActivated) (Variables|Default|GetHasValidModerationState))))

(fn ShouldDisplaySecondInteraction
  (return (Math|Boolean|ANDBoolean (Matchmaking|CanInitiatePrivateMatchmaking) (Math|Boolean|NORBoolean (Variables|Default|GetIsMatchmaking) (Variables|CodeOverride|GetLocalPlayerLookingAtScreen)) (Variables|Default|GetHasSecondInteraction) (CallFunction|IsinPlayMode) (== (Variables|UserOptions|GetJoinOption) 0) (Variables|Default|GetHasValidModerationState))))

(fn SpawnIslandDetailsUI
  (params (InteractingPlayer :type Fort%20Player%20Pawn%20Object%20Reference))
  (Networking|CanExecuteCosmeticEvents
    (:True
      (Utilities|IsValid (Variables|Default|GetIslandDetailsUI)
        (:Is%20Valid
          _)
        (:Is%20Not%20Valid
          (bind _returnvalue (UserInterface|CreateWidget "/Game/Creative/Devices/MatchmakingPortal/Matchmaking_Detail/Cr_Matchmaking_Detail.Cr_Matchmaking_Detail_C" (Fortnite|GetFortPlayerControllerfromActor InteractingPlayer)))
          (bind _output_get (Variables|Default|SetIslandDetailsUI _returnvalue))
          (Variables|/Game/Creative/Devices/MatchmakingPortal/Matchmaking_Detail/Cr_Matchmaking_Detail.Cr_Matchmaking_Detail_C|SetMatchmakingPortal self _output_get)))
      (Class|FortCreativeMatchmakingDetails|SetupInfo _output_get (Variables|Default|GetIslandLinkData))
      (if (not (Widget|IsVisible _output_get))
        (Class|FortCreativeMatchmakingDetails|Open _output_get)))))

(fn StartMatchmakingTimer
  (|CallParentFunction)
  (Default|SetUpMatchmakingInteractions))

(fn StartTimerRetrySetCurrentCode
  (if (not (Utilities|Time|IsTimerActivebyHandle (Variables|Default|GetRetrySetCodeTimerHandle)))
    (bind _returnvalue (Utilities|Time|SetTimerbyEvent (EventDispatchers|CreateEvent :!event Retry%20Set%20Current%20Code) (Variables|Default|GetTimeToRetrySetCurrentCode) true))
    (Variables|Default|SetRetrySetCodeTimerHandle _returnvalue)))

(fn UpdateCodeEntryScreen
  (params (Force%20Screen%20Update :type Boolean))
  (|CallParentFunction Force%20Screen%20Update)
  (Variables|Interaction|SetbShowSecondInteractPrompt (Interaction|ShouldDisplaySecondInteraction))
  (Variables|Interaction|SetbShowFirstInteractPrompt (Math|Boolean|ORBoolean (Variables|CodeOverride|GetLocalPlayerLookingAtScreen) (Variables|Default|GetHasValidModerationState) (not (Class|CreativeCommonDeviceFunctionLibrary|IsinPlayMode self)))))

(fn UpdateEnabled
  (params (Enabled :type Boolean))
  (|CallParentFunction Enabled)
  (Default|UpdateInWorldHUDVisibility))

(fn UpdateHUD
  (params (Matchmaking%20State :type ECreativeMatchmakingPortalState%20Enum :default "NewEnumerator0"))
  (CallFunction|UpdateHUDVisibilityBlockCheck)
  (|CallParentFunction Matchmaking%20State)
  (bind _output_get (LocalVariables|SetInWorldHUD (Default|GetinWorldHUD)))
  (Utilities|IsValid _output_get
    (:Is%20Valid
      (bind _header (CallFunction|GetHeaderDetails))
      (Class|MatchmakinginWorldDetails|SetHeaderDetails (LocalVariables|GetInWorldHUD) (select (Variables|Default|GetHasValidModerationState) _header (Variables|Default|GetModerationError)) (Variables|Default|GetHasValidModerationState))
      (Default|UpdateInWorldHUDVisibility))))

(fn UpdateinWorldHUDVisibility
  (Networking|CanExecuteCosmeticEvents
    (:True
      (if (Matchmaking|IsonPortalHudEnabled)
        (if (Class|FortPortalComponent|HasValidLinkData (Variables|Default|GetFortPortal))
          (bind _onportaldetails (Variables|Default|GetOnPortalDetails)
            (:then
              (Class|MatchmakingonPortalDetails|SetIslandTitleVisibility _onportaldetails (UserOptions|IsPortalNameVisible))
              (Class|MatchmakingonPortalDetails|SetIslandDetailsVisibility (Variables|Default|GetOnPortalDetails) (UserOptions|IsFloatingHUDVisible))
              (UserInterface|RequestRenderUpdate (Variables|Default|GetPortalWidgetComponent)))
            (:else)))
        (else
          (bind _output_get (LocalVariables|SetInWorldHUD (Default|GetinWorldHUD)))
          (Utilities|IsValid _output_get
            (:Is%20Valid
              (Class|MatchmakinginWorldDetails|SetHeaderVisible (LocalVariables|GetInWorldHUD) (UserOptions|IsPortalNameVisible))
              (Class|MatchmakinginWorldDetails|SetMatchmakingEnabledVisible (LocalVariables|GetInWorldHUD) (UserOptions|IsMatchmakingEnabledTextVisible))
              (Class|MatchmakinginWorldDetails|SetMatchmakingEnabledText (LocalVariables|GetInWorldHUD) (Utilities|Enum|Equal(Enum) (Variables|IslandDetails|GetCurrent%20Matchmaking%20State) "NewEnumerator9") :!class /Game/Creative/Devices/MatchmakingPortal/MatchmakingInWorldDetails.MatchmakingInWorldDetails_C)
              (UserInterface|RequestRenderUpdate (Variables|Default|GetInWorldHUDWidget)))))))))

(fn UpdateKeybinds
  (UserInterface|RequestRenderUpdate (Variables|Default|GetPortalWidgetComponent)))

(fn UpdatePortalInteractionifNotonCooldown
  (if (not (or (Variables|Default|GetMatchmakingOnCooldown) (Variables|Default|GetIsMatchmaking)))
    (CallFunction|UpdatePortalInteractionUIUsingLastInteracter)))

(fn UpdatePortalVisuals
  (|CallParentFunction)
  (Development|SetHiddeninGame (Variables|Default|GetNoThumbnailPlane) (or (and (or (not (Class|CreativeEnabledComponent|IsEnabled (Variables|Default|GetEnabledComponent))) (CallFunction|PortalThumbnailIsVisible)) (Variables|Default|GetHasValidModerationState)) (and (not (ChildFunctions|IsVisibleinGame)) (CallFunction|IsinPlayMode))) false))

(fn UpdateThumbnailDetails
  (params (Success :type Boolean) (Texture :type Texture%202DDynamic%20Object%20Reference))
  (Networking|CanExecuteCosmeticEvents
    (:True
      (if (CallFunction|PortalThumbnailIsVisible)
        _
        (elif (Utilities|Equal(Object) (Rendering|Material|GetMaterial (Variables|Default|GetNoThumbnailPlane) 0) (Variables|Default|GetNoThumbnailMID))
          (Default|ScaleandPositionNoThumbnailPlane)
          (else
            (Rendering|Material|SetMaterial (Variables|Default|GetNoThumbnailPlane) 0 (Variables|Default|GetNoThumbnailMID)))))
      (|CallParentFunction (Variables|GetSuccess) (Variables|GetTexture))
      (CallFunction|UpdateHUD (Variables|IslandDetails|GetCurrent%20Matchmaking%20State)))))

(fn UpdatePortalInteractionUI
  (params (InteractingPawn :type Fort%20Pawn%20Object%20Reference))
  (bind (_value _returnvalue) (Utilities|Map|Find (Variables|Default|GetClient%20Portal%20Components) (Fortnite|GetFortPlayerControllerfromActor (Variables|GetInteractingPawn))))
  (bind _interactingpawn (Variables|GetInteractingPawn))
  (Networking|CanExecuteCosmeticEvents
    (:True
      (if (HelperFunctions|IsLocalPawn _interactingpawn)
        (Variables|Default|SetLastInteractionPawn _interactingpawn)
        (if (Math|Boolean|ANDBoolean (Utilities|IsValid (Fortnite|GetFortPlayerControllerfromActor _interactingpawn) :!class /Script/Engine.KismetSystemLibrary) _returnvalue (Class|FortControllerComponentPortal|CanInitiatePortalMatchmkaing _value (Variables|Default|GetFortPortal)) (not (CallFunction|IsPartyLeaderandHasActiveReadyCheck (Fortnite|GetFortPlayerControllerfromActor _interactingpawn))))
          (bind _onportaldetails (Variables|Default|GetOnPortalDetails)
            (:then
              (Class|MatchmakingonPortalDetails|SetupInteractions _onportaldetails (Variables|GetInteractingPawn))
              (UserInterface|RequestRenderUpdate (Variables|Default|GetPortalWidgetComponent)))
            (:else))
          (else
            (|CallParentFunction)
            (CallFunction|DisablePortalInteractionUI)))))))

(fn UpdatePortalInteractionUIUsingLastInteracter
  (bind _lastinteractionpawn (Variables|Default|GetLastInteractionPawn)
    (:then
      (CallFunction|UpdatePortalInteractionUI _lastinteractionpawn))
    (:else)))

(fn ConstructionScript
  (|CallParentFunction)
  (Networking|SwitchHasAuthority
    (:Remote
      ;[1 0.6282 0 0.5] Creating MID here as the phone tool reverts back to its initial material when pasting the device (and so creating it in construction it will return to the dynamic Material)
      (bind _returnvalue (Rendering|Material|CreateDynamicMaterialInstance (Variables|Default|GetNoThumbnailPlane) 0 0 "None" :!class /Script/Engine.PrimitiveComponent))
      (Variables|Default|SetNoThumbnailMID _returnvalue))))
