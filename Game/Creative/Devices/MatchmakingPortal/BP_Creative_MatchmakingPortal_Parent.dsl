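; Parent Blueprint for the Creative matchmaking portal device, which shows a linked island's details and lets party leaders start or cancel matchmaking into it.
; FortPortal link-data callbacks feed the HUD, while the server resolves well-known names to island codes via the DAD querier.
; Matchmaking handles ready checks and a throttle timer.
; Client timers poll look-at for code entry and, optionally, line-trace to hide blocked HUDs.
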(event Custom|Cancel%20Matchmaking
  (params (Fort%20Player%20Controller :type Fort%20Player%20Controller%20Object%20Reference))
  ;[0 0.1399 1 0.5] Matchmaking Logic
  (bind _asfort_player_controller_athena (Utilities|Casting|CastToFortPlayerControllerAthena Fort%20Player%20Controller))
  (if (Class|FortPlayerController|IsPartyLeader _asfort_player_controller_athena)
    (Class|FortPlayerControllerAthena|CancelMatchmaking _asfort_player_controller_athena "CreativeCanceledByLeader")))

(event OnLinkDataError(FortPortal)
  (params (ErrorCode :type String))
  ;[0 0.6421 1 0.5] Server Replicated Island Details Loading Logic
  (Utilities|BasicCodeValidation ErrorCode
    (:Valid%20Code
      (CallFunction|IslandDataNoSuccess "NewEnumerator2"))))

(event OnLinkDataLoaded(FortPortal)
  (params (LinkData :type Fort%20Game%20Activity%20Object%20Reference))
  ;[0 0.6421 1 0.5] Server Replicated Island Details Loading Logic
  (CallFunction|IslandDataLoaded LinkData))

(event Custom|Attempt%20Matchmaking
  (params (FortPlayerController :type Fort%20Player%20Controller%20Object%20Reference))
  (bind _outputdelegate (EventDispatchers|CreateEvent :!event On%20Matchmaking%20Complete))
  ;[0 0.1399 1 0.5] Matchmaking Logic
  (bind _asfort_player_controller_athena (Utilities|Casting|CastToFortPlayerControllerAthena FortPlayerController))
  (if (Class|FortPlayerController|IsPartyLeader _asfort_player_controller_athena)
    (EventDispatchers|UnbindEventfromInGameMatchmakingComplete _asfort_player_controller_athena _outputdelegate)
    (EventDispatchers|BindEventtoInGameMatchmakingComplete _asfort_player_controller_athena _outputdelegate)
    (bind _returnvalue (CallFunction|CreateLinkId (Variables|Default|GetCurrent%20Island%20Code)))
    (bind _boutdidstartmatchmakingimmediately (Class|FortPlayerControllerAthena|PartyLeaderReadyUporStartMatchmaking _asfort_player_controller_athena _returnvalue false))
    (Class|FortPortalHUDDetails|SetActiveMatchmakingTextIndex (Variables|/Game/Creative/Devices/Common/Components/Creative_HeaderDescriptionHUD_Component.Creative_HeaderDescriptionHUD_Component_C|GetHUDWidget (Variables|Default|GetHUD%20Widget%20Component)) 1)
    (if _boutdidstartmatchmakingimmediately
      (CallFunction|StartMatchmakingTimer)
      (else
        (CallFunction|SetupWaitforReadyCheck _asfort_player_controller_athena)))))

(event OnLinkThumbnailUpdated(FortPortal)
  (params (bSuccess :type Boolean))
  ;[0 0.6421 1 0.5] Server Replicated Island Details Loading Logic
  (Networking|CanExecuteCosmeticEvents
    (:True
      (CallFunction|UpdateThumbnailDetails bSuccess (Variables|/Script/FortniteGame.FortPortalComponent|GetThumbnailTexture (Variables|Default|GetFortPortal))))))

(event EventOnMinigameEnded
  (|CallParentFunction)
  (if (Variables|Default|GetbExpireLinkCodeLockOnMinigameEnd)
    (Class|FortPortalComponent|ExpireLinkCodeLock (Variables|Default|GetFortPortal))))

(event EventInitializeDevice
  ;[0 0.05991 0.4896] Updating HUD When the Device changes.
  (|CallParentFunction)
  (CodeOverride|UpdateLockMode)
  (Networking|CanExecuteCosmeticEvents
    (:True
      (CallFunction|UpdateLight)
      (CallFunction|StoreClientPortalComponents)
      (PartyContext|BindEventtoOnLocalPlayerStateChanged (BlueprintContext|GetFortPartyContext) (delegate Custom|Local%20Player%20State%20Changed))
      (bind _outvolumemanager (Volume|GetVolumeManager)
        (:Valid
          (EventDispatchers|BindEventtoOnClientExitVolume _outvolumemanager (EventDispatchers|CreateEvent :!event Custom|OnClientExitVolume))
          ;;[1 1 1] For DAD Service, Clients Update through WellKnownNameCode replication, so we won't update from Properties Override
          (if (CallFunction|IsWellKnownNameSupported)
            _
            (else
              _)))
        (:Invalid
          ;;[1 1 1] For DAD Service, Server Query to DAD_IslandQuerier to get the code
          (if (CallFunction|IsWellKnownNameSupported)
            (bind _link_id (CallFunction|MakeSafeOnlineLinkCode (Variables|UserOptions|GetIslandCode_Knob)))
            (bind _returnvalue (CallFunction|LinkIdtoString _link_id))
            (if (Utilities|String|NotEqualExactly(String) (CallFunction|DADIslandQuerierGetIslandCode _returnvalue) (Variables|DAD|GetWellKnownNameCode))
              (Variables|DAD|SetbIslandCodeFromDAD (.IslandCodeFound (CallFunction|DADIslandQuerierGetIslandCode _returnvalue)))
              (Variables|DAD|SetWellKnownNameCode (CallFunction|DADIslandQuerierGetIslandCode _returnvalue)))
            (else
              _))))
      (bind _link_id_2 (CallFunction|MakeSafeOnlineLinkCode (Variables|UserOptions|GetIslandCode_Knob)))
      (bind _returnvalue_2 (CallFunction|LinkIdtoString _link_id_2))
      (CallFunction|UpdateIslandDetails _returnvalue_2)
      (CallFunction|StartOrEndQuery))
    (:False
      (Class|FortPortalComponent|SetResetCodeAfterCooldown (Variables|Default|GetFortPortal) (Utilities|Select :Index (Variables|Default|GetCodeStatusAfterCooldown) :Option%200 true :Option%201 false)))))

(event Custom|Local%20Player%20State%20Changed
  (params (PlayerInfo :type Fort%20Team%20Member%20Info%20Structure%20%28by%20ref%29))
  ;[0 0.05991 0.4896] Updating HUD When the Device changes.
  (CallFunction|UpdateHUD (Variables|IslandDetails|GetCurrent%20Matchmaking%20State)))

(event Custom|Enable
  (params (Instigator :type Controller%20Object%20Reference))
  ;[0.6961 0 1 0.5] Recievers
  (Class|CreativeEnabledComponent|SetEnabled (Variables|Default|GetEnabledComponent) true))

(event OnComponentEndOverlap(HUDVisibilityVolume)
  (params (OverlappedComponent :type Primitive%20Component%20Object%20Reference) (OtherActor :type Actor%20Object%20Reference) (OtherComp :type Primitive%20Component%20Object%20Reference) (OtherBodyIndex :type Integer))
  ;[1 0.7039 0 0.5] BeginOverlap / EndOverlap: When you're close
  ; A.Show or Hide HUD
  ; B.Start or End Visibility Blocking Check
  (Utilities|FlowControl|Sequence
    (:then_0
      (CallFunction|SetHUDVisibility OtherActor)
      (CodeOverride|LocalPlayerRangeCheck OtherActor))
    (:then_1
      (if (Variables|VisibilityCheck|GetIsHUDVisibilityBlockCheckEnabled)
        (VisibilityCheck|EndHUDVisibilityBlockingCheck)))))

(event OnComponentBeginOverlap(HUDVisibilityVolume)
  (params (OverlappedComponent :type Primitive%20Component%20Object%20Reference) (OtherActor :type Actor%20Object%20Reference) (OtherComp :type Primitive%20Component%20Object%20Reference) (OtherBodyIndex :type Integer) (bFromSweep :type Boolean) (SweepResult :type Hit%20Result%20Structure%20%28by%20ref%29))
  ;[1 0.7039 0 0.5] BeginOverlap / EndOverlap: When you're close
  ; A.Show or Hide HUD
  ; B.Start or End Visibility Blocking Check
  (Utilities|FlowControl|Sequence
    (:then_0
      (CodeOverride|LocalPlayerRangeCheck OtherActor true))
    (:then_1
      (if (Variables|VisibilityCheck|GetIsHUDVisibilityBlockCheckEnabled)
        (VisibilityCheck|StartHUDVisibilityBlockingCheck)
        (CallFunction|UpdateHUDVisibilityBlockCheck)
        (else
          (CallFunction|SetHUDVisibility OtherActor true))))))

(event Custom|Disable
  (params (Instigator :type Controller%20Object%20Reference))
  ;[0.6961 0 1 0.5] Recievers
  (Class|CreativeEnabledComponent|SetEnabled (Variables|Default|GetEnabledComponent) false))

(event OnEnabledStateChanged(EnabledComponent)
  (params (Enabled :type Boolean))
  ;[0.6961 0 1 0.5] Recievers
  (EnabledState|UpdateEnabled Enabled))

(event Custom|UpdateHUDVisibilityBlockCheck
  ;[1 1 1] UpdateVisibilityBlockCheck
  (for _array_element (Variables|CodeOverride|GetLocal%20Pawns%20in%20Range)
    (CallFunction|SetHUDVisibility _array_element (not (VisibilityCheck|IsThereAVisibilityBlocktoHUD _array_element (Variables|VisibilityCheck|GetMeshTargetToCheckHUDVisibility))))))

(event EventBeginPlay
  ;[1 1 1] Begin/End Play
  (|CallParentFunction)
  (CallFunction|SetupDevice))

(event Custom|OnWellKnownIslandsUpdated
  (bind _wellknownnamecode (Variables|DAD|GetWellKnownNameCode))
  ;[1 1 1] OnDataAssetDirectoryUpdate
  (Development|PrintString "OnWellKnownIslandsUpdated BP" true true "(R=0.000000,G=0.660000,B=1.000000,A=1.000000)" 1e+01)
  (Networking|SwitchHasAuthority
    (:Authority
      (bind _link_id (CallFunction|MakeSafeOnlineLinkCode (Variables|UserOptions|GetIslandCode_Knob)))
      (bind _returnvalue (CallFunction|LinkIdtoString _link_id))
      (Development|PrintString (Utilities|String|Append "DataAssetDirectoryUpdate Success: " "-- Island Knob: " _returnvalue " -- Previous WellKnownNameCode: " _wellknownnamecode "-- New Code:" (CallFunction|DADIslandQuerierGetIslandCode _returnvalue)) true true "(R=0.000000,G=0.660000,B=1.000000,A=1.000000)" 1e+01)
      (if (Utilities|String|NotEqualExactly(String) (CallFunction|DADIslandQuerierGetIslandCode _returnvalue) _wellknownnamecode)
        (Development|PrintString "Updated Well Known Name Code")
        (Variables|DAD|SetbDADRequestedUpdate true)
        (Variables|DAD|SetbIslandCodeFromDAD (.IslandCodeFound (CallFunction|DADIslandQuerierGetIslandCode _returnvalue)))
        (Variables|DAD|SetWellKnownNameCode (CallFunction|DADIslandQuerierGetIslandCode _returnvalue))
        (Variables|DAD|SetbDADRequestedUpdate false)))))

(event Custom|Open%20Code%20Override%20HUD
  (params (Pawn :type Pawn%20Object%20Reference))
  ;[0 0.6904 1] Client Code Override Functions
  (bind _class (Utilities|AsyncLoadClassAsset (Variables|CodeOverride|GetUI_IslandCodeEntrySoftClass))
    (:Completed
      (if (Utilities|IsValidClass _class)
        (bind _asuser_widget (Utilities|Casting|CastToUserWidgetClass _class))
        (bind _returnvalue (UserInterface|CreateWidget _asuser_widget (Fortnite|GetFortPlayerControllerfromActor Pawn)))
        (Init|InitfromObject(Message) _returnvalue self)))))

(event Custom|Query%20Player%20Look%20At
  ;[0 0.6904 1] Client Code Override Functions
  (CodeOverride|UpdateCodeEntryScreen false))

(event Custom|Start%20or%20End%20Query
  ;[0 0.6904 1] Client Code Override Functions
  (Networking|CanExecuteCosmeticEvents
    (:True
      (if (Math|Boolean|ANDBoolean (CallFunction|ConsoleIsVisible) (CallFunction|IsinPlayMode) (Math|Conversions|ToBoolean(Integer) (Utilities|Array|Length (Variables|CodeOverride|GetLocal%20Pawns%20in%20Range))))
        (bind _returnvalue (Utilities|Time|SetTimerbyEvent (delegate Custom|Query%20Player%20Look%20At) (Variables|CodeOverride|GetLook%20At%20Update%20Time) true))
        (Variables|CodeOverride|SetLook%20at%20Timer _returnvalue)
        (else
          (Utilities|Time|ClearandInvalidateTimerbyHandle (Variables|CodeOverride|GetLook%20at%20Timer))
          (CodeOverride|UpdateCodeEntryScreen false))))))

(event Custom|Local%20Player%20Entered%20Range
  (params (Pawn :type Pawn%20Object%20Reference))
  ;[0 0.6904 1] Client Code Override Functions
  (Utilities|Array|AddUnique (Variables|CodeOverride|GetLocal%20Pawns%20in%20Range) Pawn)
  (CallFunction|StartOrEndQuery))

(event Custom|Local%20Player%20Exited%20Range
  (params (Pawn :type Pawn%20Object%20Reference))
  ;[0 0.6904 1] Client Code Override Functions
  (Utilities|Array|RemoveItem (Variables|CodeOverride|GetLocal%20Pawns%20in%20Range) Pawn)
  (CallFunction|StartOrEndQuery))

(event ChildFunctions|EventPlayModeChanged
  (params (Minigame :type Fort%20Minigame%20Object%20Reference) (bIsInPlayMode :type Boolean))
  ;[0 0.6904 1] Client Code Override Functions
  ;;[0 0.006206 0.06771 0.9] Query whether to start checking against screen when entering play mode
  ;; (necessary for when a pawn starts within the range during edit mode)
  (CallFunction|StartOrEndQuery))

(event OnPortalLinkCodeLockStatusChanged(FortPortal)
  (params (Status :type EPortalLinkCodeLockStatus%20Enum :default "Unlocked_NotSet"))
  ;[0 0.6904 1] Client Code Override Functions
  (CodeOverride|UpdateCodeEntryScreen true))

(event Custom|OnClientExitVolume
  (params (Client :type Player%20State%20Object%20Reference) (Volume :type Fort%20Volume%20Object%20Reference))
  ;[1 1 1] When anyone leaves the volume, cancel matchmaking for their whole party
  (if (Variables|Default|GetbLeavingIslandCancelsMatchmaking)
    (bind _asfort_player_controller_athena (Utilities|Casting|CastToFortPlayerControllerAthena (Actor|GetOwner Client)))
    (Class|FortPlayerControllerAthena|CancelMatchmaking _asfort_player_controller_athena "CreativeMemberLeftIsland")
    (bind _outvolumemanager (Volume|GetVolumeManager)
      (:Valid
        (EventDispatchers|UnbindEventfromOnClientExitVolume _outvolumemanager (EventDispatchers|CreateEvent :!event Custom|OnClientExitVolume))))))

(event EventDestroyed
  ;[1 1 1] Destroyed - Unbind functions
  (|CallParentFunction)
  (Networking|CanExecuteCosmeticEvents
    (:True
      (bind _outvolumemanager (Volume|GetVolumeManager)
        (:Valid
          (EventDispatchers|UnbindEventfromOnClientExitVolume _outvolumemanager (EventDispatchers|CreateEvent :!event Custom|OnClientExitVolume)))))))

(event EventEndPlay
  (params (EndPlayReason :type EEndPlayReason%20Enum :default "Destroyed"))
  ;[1 1 1] End Play
  (if (CallFunction|IsWellKnownNameSupported)
    (CallFunction|BindorUnbindtoIslandQuerier false)
    (else
      _))
  (|CallParentFunction EndPlayReason))

(fn AthenaInteractDetails
  (return false false))

(fn BindOrUnbindToIslandQuerier
  (params (Bind :type Boolean))
  (bind _outputdelegate (EventDispatchers|CreateEvent :!event OnWellKnownIslandsUpdated))
  (Networking|SwitchHasAuthority
    (:Authority
      (bind _asfort_game_instance (Utilities|Casting|CastToFortGameInstance (Game|GetGameInstance)))
      (Utilities|IsValid (Class|FortGameInstance|GetDADIslandQuerier _asfort_game_instance)
        (:Is%20Valid
          (if (Variables|GetBind)
            (EventDispatchers|BindEventtoOnWellKnownIslandsUpdated (Class|FortGameInstance|GetDADIslandQuerier _asfort_game_instance) _outputdelegate)
            (else
              (EventDispatchers|UnbindEventfromOnWellKnownIslandsUpdated (Class|FortGameInstance|GetDADIslandQuerier _asfort_game_instance) _outputdelegate))))))))

(fn BlueprintCanInteract
  (params (InteractingPawn :type Fort%20Pawn%20Object%20Reference) (InteractionBeingAttempted :type EInteractionBeingAttempted%20Enum :default "FirstInteraction") (InteractionType :type TInteractionType%20Enum :default "IT_NoInteraction"))
  (bind _returnvalue (Fortnite|GetFortPlayerControllerfromActor InteractingPawn))
  (Networking|CanExecuteCosmeticEvents
    (:True
      (if (Interaction|CanEditOptionsComponent (Variables|Default|GetToyOptionsComponent) (Variables|GetInteractingPawn))
        (return true)
        (elif (and (Class|CreativeEnabledComponent|IsEnabled (Variables|Default|GetEnabledComponent)) (CallFunction|CanInteract))
          (if (Variables|CodeOverride|GetLocalPlayerLookingAtScreen)
            (return (Utilities|Enum|NotEqual(Enum) (Variables|/Script/FortniteGame.FortPortalComponent|GetLinkCodeLockStatus (Variables|Default|GetFortPortal)) "Locked"))
            (else
              (Utilities|IsValid _returnvalue
                (:Is%20Valid
                  (return (and (or (Class|FortPortalComponent|IsPlayerEligibletoBeginMatchmaking (Variables|Default|GetFortPortal) _returnvalue) (and (Class|FortPlayerController|IsPartyLeader _returnvalue) (Class|FortControllerComponentPortal|CanCancelPortalMatchmaking (Utilities|Map|Find (Variables|Default|GetClient%20Portal%20Components) _returnvalue)))) (or (not (Variables|Default|GetbPrivateIsland)) (Utilities|String|EqualExactly(String) (Class|FortPortalComponent|GetAccountIdOwnerOfIsland (Variables|Default|GetFortPortal)) (Variables|Default|GetCreator%20Id))))))
                (:Is%20Not%20Valid
                  (return false)))))
          (else
            (return false)))))))

(fn BlueprintGetFailedInteractionString
  (params (InteractingPawn :type Fort%20Pawn%20Object%20Reference) (InteractionBeingAttempted :type EInteractionBeingAttempted%20Enum :default "FirstInteraction"))
  (return (select (Class|FortPortalComponent|IsPartyEligibleforMatchmaking (Variables|Default|GetFortPortal)) "" (Variables|Default|GetParty%20Ineligible%20Text))))

(fn BlueprintGetInteractionString
  (params (InteractingPawn :type Fort%20Pawn%20Object%20Reference) (InteractionBeingAttempted :type EInteractionBeingAttempted%20Enum :default "FirstInteraction"))
  (bind _interactingpawn (Variables|GetInteractingPawn))
  (bind _returnvalue (Fortnite|GetFortPlayerControllerfromActor _interactingpawn))
  (bind _client_portal_components (Variables|Default|GetClient%20Portal%20Components))
  (bind (_value _returnvalue_2) (Utilities|Map|Find _client_portal_components _returnvalue))
  (Networking|CanExecuteCosmeticEvents
    (:True
      (if (Interaction|CanEditOptionsComponent (Variables|Default|GetToyOptionsComponent) InteractingPawn)
        (bind _returnvalue_3 (|CallParentFunction InteractingPawn))
        (return _returnvalue_3)
        (elif (Variables|CodeOverride|GetLocalPlayerLookingAtScreen)
          (return (Variables|CodeOverride|GetSet%20Code%20Text))
          (else
            (Utilities|IsValid _returnvalue
              (:Is%20Valid
                (if _returnvalue_2
                  (if (CallFunction|IsPartyLeaderandHasActiveReadyCheck _returnvalue)
                    (return (Variables|Default|GetStop%20Matchmaking%20Text))
                    (elif (Class|FortControllerComponentPortal|CanInitiatePortalMatchmkaing _value (Variables|Default|GetFortPortal))
                      (return (Variables|Default|GetStart%20Matchmaking%20Text))
                      (elif (Class|FortControllerComponentPortal|CanCancelPortalMatchmaking _value)
                        (return (Variables|Default|GetStop%20Matchmaking%20Text))
                        (else
                          (return "INVTEXT(\" \")"))))))))))))))

(fn BlueprintGetInteractionTime
  (params (InteractingPawn :type Fort%20Pawn%20Object%20Reference) (InteractionBeingAttempted :type EInteractionBeingAttempted%20Enum :default "FirstInteraction"))
  (return (select (or (Interaction|CanEditOptionsComponent (Variables|Default|GetToyOptionsComponent) InteractingPawn) (Variables|CodeOverride|GetLocalPlayerLookingAtScreen)) 0.0 (Variables|Default|GetMatchmaking%20Interact%20Time)) true))

(fn BlueprintOnLocalInteract
  (params (InteractingPawn :type Fort%20Player%20Pawn%20Object%20Reference) (InteractionBeingAttempted :type EInteractionBeingAttempted%20Enum :default "FirstInteraction"))
  (bind _local_pc (LocalVariables|Getlocal%20PC))
  (bind _client_portal_components (Variables|Default|GetClient%20Portal%20Components))
  (bind (_value _returnvalue) (Utilities|Map|Find _client_portal_components _local_pc))
  (bind _local_pc_2 (LocalVariables|Getlocal%20PC))
  (Networking|CanExecuteCosmeticEvents
    (:True
      (if (Class|CreativeEnabledComponent|IsEnabled (Variables|Default|GetEnabledComponent))
        (if (Interaction|CanEditOptionsComponent (Variables|Default|GetToyOptionsComponent) (Variables|GetInteractingPawn))
          ;[1 1 1] Options Customize Menu
          (bind _returnvalue_2 (Interaction|OnLocalInteract (Variables|Default|GetToyOptionsComponent) (Variables|GetInteractingPawn)))
          (return _returnvalue_2)
          (elif (Variables|CodeOverride|GetLocalPlayerLookingAtScreen)
            ;[1 1 1] Local Code Override UI
            (CallFunction|OpenCodeOverrideHUD (Variables|GetInteractingPawn))
            (return true)
            (else
              ;[1 1 1] New matchmaking portal UI interaction
              (bind _output_get (LocalVariables|Setlocal%20PC (Fortnite|GetFortPlayerControllerfromActor (Variables|GetInteractingPawn))))
              (if (Class|HeaderDescriptionHUDComponent|IsNewMatchmakingPortalUIEnabled)
                (bind _asbp_creative_matchmaking_portal (Utilities|Casting|CastToBP_Creative_MatchmakingPortal self))
                (Class|HeaderDescriptionHUDComponent|PortalInteract (Variables|Default|GetHUD%20Widget%20Component) (Utilities|Struct|MakePortalMatchmakingData (Variables|/Game/Creative/Devices/MatchmakingPortal/BP_Creative_MatchmakingPortal.BP_Creative_MatchmakingPortal_C|GetLastInteractionPawn _asbp_creative_matchmaking_portal) (Utilities|Map|Find (Variables|Default|GetClient%20Portal%20Components) (LocalVariables|Getlocal%20PC)) (Variables|Default|GetFortPortal) (Variables|/Game/Creative/Devices/MatchmakingPortal/BP_Creative_MatchmakingPortal.BP_Creative_MatchmakingPortal_C|GetIslandLinkData _asbp_creative_matchmaking_portal)))
                (else
                  ;[1 1 1] Attempt Matchmake / Cancel Matchmake Attempt
                  (Utilities|IsValid _output_get
                    (:Is%20Valid
                      (if _returnvalue
                        (if (not (CallFunction|IsPartyLeaderandHasActiveReadyCheck (LocalVariables|Getlocal%20PC)))
                          (if (Class|FortControllerComponentPortal|CanInitiatePortalMatchmkaing _value (Variables|Default|GetFortPortal))
                            (CallFunction|AttemptMatchmaking _local_pc_2)
                            (elif (Class|FortControllerComponentPortal|CanCancelPortalMatchmaking _value)
                              (CallFunction|CancelMatchmaking _local_pc_2)
                              (else
                                (return false))))
                          (else
                            (CallFunction|CancelReadyCheck (LocalVariables|Getlocal%20PC))))
                        (Variables|Default|SetLastInteractTime (Utilities|Time|GetGameTimeinSeconds)) ; Saving interact time to pause \ninteraction to avoid flipflopping \nbetween states.
                        (return true)))))))))
        (else
          (return false))))))

(fn CancelReadyCheck
  (params (InteractionPlayerController :type Fort%20Player%20Controller%20Object%20Reference))
  (if (CallFunction|IsPartyLeaderandHasActiveReadyCheck (Variables|GetInteractionPlayerController))
    (bind _asfort_player_controller_athena (Utilities|Casting|CastToFortPlayerControllerAthena (Variables|GetInteractionPlayerController)))
    (Class|FortPlayerControllerAthena|CancelCreativeMatchmakingReadyCheck _asfort_player_controller_athena)))

(fn CanInteract
  (return (> (Utilities|Time|GetGameTimeinSeconds) (+ (Variables|Default|GetLastInteractTime) (Variables|Default|GetInteractDelayAfterSuccess)))))

(fn CanPlayPortalAudio
  (return (not (Audio|Components|Audio|IsPlaying (Variables|Default|GetPortal%20Loop%20Audio)))))

(fn ConsoleIsVisible
  (return (Utilities|Enum|NotEqual(Enum) (Variables|/Script/FortniteGame.FortPortalComponent|GetLinkCodeLockMode (Variables|Default|GetFortPortal)) "AlwaysLocked")))

(fn DAD_IslandQuerierGetIslandCode
  (params (WellKnownName :type String))
  (bind _wellknownname (Variables|GetWellKnownName))
  (bind _asfort_game_instance (Utilities|Casting|CastToFortGameInstance (Game|GetGameInstance))
    (:then
      (Utilities|IsValid (Class|FortGameInstance|GetDADIslandQuerier _asfort_game_instance)
        (:Is%20Valid
          (if (not (Utilities|String|IsEmpty (Class|DADIslandQuerier|GetIslandCodebyWellKnownName (Class|FortGameInstance|GetDADIslandQuerier _asfort_game_instance) _wellknownname)))
            ;[1 1 1 1 12] Return Island Code by Well Known Name
            (return (Class|DADIslandQuerier|GetIslandCodebyWellKnownName (Class|FortGameInstance|GetDADIslandQuerier _asfort_game_instance) _wellknownname) true)))
        (:Is%20Not%20Valid
          _))
      ;[1 1 1] Return Original String
      (return (Variables|GetWellKnownName) false))
    (:CastFailed
      _)))

(fn EndHUDVisibilityBlockingCheck
  (if (== (Utilities|Array|Length (Variables|CodeOverride|GetLocal%20Pawns%20in%20Range)) 0)
    (Utilities|Time|ClearandInvalidateTimerbyHandle (Variables|VisibilityCheck|GetHUDVisibilityCheckTimer))))

(fn GetDescriptionDetails
  (return ""))

(fn GetHeaderDetails
  (return (select (Variables|CodeOverride|GetLocalPlayerLookingAtScreen) (Variables|CodeOverride|GetLookingAtHeader) (Variables|IslandDetails|GetIsland%20Name)) (Transformation|GetWorldLocation (select (Variables|CodeOverride|GetLocalPlayerLookingAtScreen) (Variables|Default|GetScreen%20Look%20At%20Point) (Variables|Default|GetPortal%20Look%20at%20Point))) 0))

(fn GetInteractionDetails
  (params (Controller :type Fort%20Player%20Controller%20Object%20Reference))
  ;[1 1 1] Run through the various matchmaking criteria to figure out what the interaction state is between the player and the portal.
  (Networking|CanExecuteCosmeticEvents
    (:True
      (Utilities|IsValid Controller
        (:Is%20Valid
          (if (Variables|CodeOverride|GetLocalPlayerLookingAtScreen)
            (return (select (Utilities|Enum|Equal(Enum) (Variables|/Script/FortniteGame.FortPortalComponent|GetLinkCodeLockStatus (Variables|Default|GetFortPortal)) "Locked") "NewEnumerator5" "NewEnumerator4") (Class|FortPortalComponent|GetRemainingLockDuration (Variables|Default|GetFortPortal)))
            (elif (Class|FortPortalComponent|IsPartyEligibleforMatchmaking (Variables|Default|GetFortPortal))
              (return (select (Class|FortPlayerController|IsPartyLeader Controller) "NewEnumerator0" "NewEnumerator2") 0)
              (else
                (return (select (Utilities|Enum|Equal(Enum) (Variables|IslandDetails|GetCurrent%20Matchmaking%20State) "NewEnumerator3") "NewEnumerator3" "NewEnumerator0") 0)))))))))

(fn GetIslandPortalDetails
  (return (select (Variables|CodeOverride|GetLocalPlayerLookingAtScreen) "NewEnumerator0" (Variables|IslandDetails|GetCurrent%20Matchmaking%20State)) (Variables|IslandDetails|GetCreator%20Name) (Variables|IslandDetails|GetDescription%20Text) (Utilities|Text|ToText(String) (CallFunction|GetIslandLinkCodeWithoutVersioning)) (Variables|IslandDetails|GetPopMin) (Variables|IslandDetails|GetPopMax)))

(fn GetFocusedSocketLocation
  (return (Transformation|GetWorldLocation (select (Variables|CodeOverride|GetLocalPlayerLookingAtScreen) (Variables|Default|GetScreen%20Look%20At%20Point) (Variables|Default|GetPortal%20Look%20at%20Point)))))

(fn GetIslandLinkCodeWithoutVersioning
  (bind _outlinkid (Class|CreativeBlueprintLibrary|MakeOnlineLinkCode (Variables|Default|GetCurrent%20Island%20Code) true))
  (bind _returnvalue (CallFunction|LinkIdtoString _outlinkid))
  (return _returnvalue))

(fn GetScaleMax
  (params (DefaultScaleMax :type Float%20%28single-precision%29))
  ;[1 1 1 0.7] Default is 5.0
  (return (Variables|Default|GetMaxDeviceScale)))

(fn GetScaleMin
  (params (DefaultScaleMin :type Float%20%28single-precision%29))
  ;[1 1 1 0.7] Default is 0.33
  (return (Variables|Default|GetMinDeviceScale)))

(fn HandleMatchmakingThrottleTimerComplete
  (Variables|Default|SetMatchmakingOnCooldown false)
  (Variables|Interaction|SetbAllowInteract true))

(fn IsPartyLeaderAndHasActiveReadyCheck
  (params (PlayerController :type Fort%20Player%20Controller%20Object%20Reference))
  (Utilities|IsValid PlayerController
    (:Is%20Valid
      (if (Class|FortPlayerController|IsPartyLeader PlayerController)
        (bind _asfort_player_controller_athena (Utilities|Casting|CastToFortPlayerControllerAthena (Variables|GetPlayerController)))
        (return (Class|FortPlayerControllerAthena|IsPartyMemberReadyCheckStatusinProgress _asfort_player_controller_athena))
        (else
          _)))
    (:Is%20Not%20Valid
      _))
  (return false))

(fn IslandDataLoaded
  (params (LinkData :type Fort%20Game%20Activity%20Object%20Reference%20%28by%20ref%29))
  (bind (_minimumnumberofplayers _maximumnumberofplayers _mmstype _mmsprivacy) (Utilities|Struct|BreakCreativeIslandMatchmakingSettings (Class|FortGameActivity|GetMatchmakingSettingsV1 LinkData)))
  (bind _returnvalue (Class|FortGameActivity|GetMatchmakingSettingsV1 LinkData))
  (bind _returnvalue_2 (Class|FortGameActivity|GetCreativeAccountId LinkData))
  (bind _linkdata (Variables|GetLinkData))
  (bind _returnvalue_3 (CallFunction|LinkIdtoString (Variables|/Script/FortniteGame.FortGameActivity|GetLinkId LinkData)))
  (if (> (Utilities|String|Len _returnvalue_3) 0)
    (Networking|CanExecuteCosmeticEvents
      (:True
        (Variables|Default|SetCreator%20Id _returnvalue_2)
        (if (FortActivityLibrary|ShouldDisplayCreatorCodeasDisplayName _returnvalue_2)
          (Variables|IslandDetails|SetCreator%20Name (Utilities|Text|ToText(String) (Variables|/Script/FortniteGame.FortGameActivity|GetSupportCode LinkData)))
          (else
            (Variables|IslandDetails|SetCreator%20Name (Utilities|Text|ToText(String) (Variables|/Script/FortniteGame.FortGameActivity|GetCreator LinkData)))))
        (Variables|IslandDetails|SetIsland%20Name (Utilities|Text|ToText(String) (Variables|/Script/FortniteGame.FortGameActivity|GetTitle _linkdata)))
        (Variables|IslandDetails|SetDescription%20Text (Utilities|Text|ToText(String) (Variables|/Script/FortniteGame.FortGameActivity|GetTagline _linkdata)))
        (Variables|IslandDetails|SetPopMin _minimumnumberofplayers)
        (Variables|IslandDetails|SetPopMax _maximumnumberofplayers)
        (Default|CallOnDeviceActivated)
        (if (and (Utilities|String|EqualExactly(String) _mmsprivacy "Private") (Utilities|String|NotEqualExactly(String) _returnvalue_2 (Class|FortPortalComponent|GetAccountIdOwnerOfIsland (Variables|Default|GetFortPortal))))
          (Variables|Default|SetbPrivateIsland true)
          (CallFunction|UpdateHUD "NewEnumerator9")
          (else
            (Variables|Default|SetbPrivateIsland false)
            (CallFunction|UpdateHUD "NewEnumerator3")))))
    (else
      (CallFunction|IslandDataNoSuccess "NewEnumerator4"))))

(fn IslandDataNoSuccess
  (params (Matchmaking%20State :type ECreativeMatchmakingPortalState%20Enum :default "NewEnumerator0"))
  (Networking|CanExecuteCosmeticEvents
    (:True
      (Variables|IslandDetails|SetIsland%20Name (Variables|/Script/FortniteGame.FortActorOptionsComponent|GetInternal_OptionsDisplayName (Variables|Default|GetToyOptionsComponent)))
      (Variables|IslandDetails|SetThumbnailTexture "")
      (CallFunction|UpdateHUD Matchmaking%20State)
      (CallFunction|UpdatePortalVisuals))))

(fn IsThereAVisibilityBlockToHUD
  (params (ActorFrom :type Actor%20Object%20Reference) (MeshTarget :type Static%20Mesh%20Component%20Object%20Reference))
  (bind (_outhit _returnvalue) (Collision|LineTraceByChannel (Transformation|GetActorLocation ActorFrom) (Transformation|GetWorldLocation MeshTarget) "TraceTypeQuery1" false (Utilities|Array|MakeArray ActorFrom)))
  (return _returnvalue))

(fn IsWellKnownNameSupported
  (bind _fortportal (Variables|Default|GetFortPortal))
  (Utilities|IsValid _fortportal
    (:Is%20Valid
      (return (and (Variables|DAD|GetbSupportsWellKnownName) (Class|FortPortalComponent|IsWellKnownNameSupported _fortportal))))
    (:Is%20Not%20Valid
      (return (Variables|DAD|GetbSupportsWellKnownName)))))

(fn LocalPlayerRangeCheck
  (params (Actor :type Actor%20Object%20Reference) (Entered%20Range :type Boolean))
  (bind _returnvalue (Fortnite|GetFortPlayerControllerfromActor Actor))
  (Networking|CanExecuteCosmeticEvents
    (:True
      (Utilities|IsValid _returnvalue
        (:Is%20Valid
          (bind _aspawn (Utilities|Casting|CastToPawn Actor))
          (if (Utilities|Array|ContainsItem (Fortnite|GetLocalFortPlayerControllers true) _returnvalue)
            (if (Variables|GetEntered%20Range)
              (CallFunction|LocalPlayerEnteredRange _aspawn)
              (else
                (CallFunction|LocalPlayerExitedRange _aspawn)))))))))

(fn MakeSafeOnlineLinkCode
  (params (Link%20Code :type String))
  (bind _outlinkid (Class|CreativeBlueprintLibrary|MakeOnlineLinkCode Link%20Code true))
  (return _outlinkid))

(fn MatchmakingPortalHasSecondInteraction
  (params (InteractingPawn :type Fort%20Player%20Pawn%20Object%20Reference))
  (return false))

(fn OnMatchmakingComplete
  (params (bSuccess :type Boolean))
  (if (Variables|Default|GetMatchmakingOnCooldown)
    (Variables|Interaction|SetbAllowInteract false)))

(fn OnRep_WellKnownNameCode
  (CallFunction|UpdateIslandDetails (Variables|DAD|GetWellKnownNameCode))
  (CallFunction|StartOrEndQuery))

(fn OnUserGeneratedContentRestrictionChanged
  (params (bEnabled :type Boolean))
  (if bEnabled
    (Actor|DestroyActor)))

(fn PlayPortalAudio
  (if (CallFunction|CanPlayPortalAudio)
    (Audio|Components|Audio|FadeIn (Variables|Default|GetPortal%20Loop%20Audio) 0.5 1.0 0.0 "Linear" :!class /Script/Engine.AudioComponent)
    (return)
    (else
      (return))))

(fn PortalHUDIsVisible
  (params (bVisibility :type Boolean))
  (return (and (Variables|GetbVisibility) (Class|CreativeEnabledComponent|IsEnabled (Variables|Default|GetEnabledComponent)))))

(fn PortalThumbnailIsVisible
  (return (Utilities|Select :Index (Variables|IslandDetails|GetCurrent%20Matchmaking%20State) :NewEnumerator0 false :NewEnumerator1 false :NewEnumerator2 false :NewEnumerator3 (Math|Boolean|ANDBoolean (Class|CreativeEnabledComponent|IsEnabled (Variables|Default|GetEnabledComponent)) (Utilities|IsValid (Variables|IslandDetails|GetThumbnailTexture) :!class /Script/Engine.KismetSystemLibrary) (Variables|LegacyOptions|GetbIsVisible)) :NewEnumerator4 false :NewEnumerator9 false)))

(fn ReadyCheckCompleted
  (bind _readycheckinteractionplayer (Variables|Default|GetReadyCheckInteractionPlayer)
    (:then
      (EventDispatchers|UnbindEventfromInGameMatchmakingReadyCheckComplete _readycheckinteractionplayer (EventDispatchers|CreateEvent :!event ReadyCheckCompleted))
      (bind _returnvalue (CallFunction|CreateLinkId (Variables|Default|GetCurrent%20Island%20Code)))
      (Class|FortPlayerControllerAthena|StartMatchmakingbyLinkCode _readycheckinteractionplayer _returnvalue (and (Variables|Default|GetbPrivateMatchmaking) (Matchmaking|CanInitiatePrivateMatchmaking)))
      (CallFunction|StartMatchmakingTimer))
    (:else)))

(fn SendIslandLink
  (params (Requesting%20Player%20Controller :type Fort%20Player%20Controller%20Object%20Reference) (Link :type Fort%20Creative%20Island%20Link%20Object%20Reference))
  (bind (_mnemonic _version) (Utilities|Struct|BreakOnlineLinkId (Variables|/Script/FortniteUI.FortCreativeIslandLink|GetLinkId Link)))
  (Utilities|IsValid Link
    (:Is%20Valid
      (CallFunction|UpdateIslandDetails (Class|FortPortalComponent|MakeLinkCodewithVersioning (Variables|Default|GetFortPortal) _mnemonic _version) true Requesting%20Player%20Controller)
      (Variables|Default|SetLastInteractTime (Utilities|Time|GetGameTimeinSeconds)) ; Saving interact time. Used to disable interaction\nfor a short time.
      (return true))
    (:Is%20Not%20Valid
      (return false))))

(fn SetHUDVisibility
  (params (Overlapping%20Actor :type Actor%20Object%20Reference) (bVisibility :type Boolean))
  ;[1 0.7441 0] Shows or Hides the Widget if you get close enough  / far enough away from it
  (Networking|CanExecuteCosmeticEvents
    (:True
      (Variables|Default|GetHUD%20Widget%20Component
        (:then
          (Class|CreativeHeaderDescriptionHUDComponent|ShowWidget (Variables|Default|GetHUD%20Widget%20Component) Overlapping%20Actor (CallFunction|PortalHUDIsVisible (Variables|GetbVisibility))))
        (:else)))))

(fn SetupDevice
  (bind _self self)
  (Utilities|FlowControl|Sequence
    (:then_0
      (VisibilityCheck|SetupMeshTargettoCheckVisibility))
    (:then_1
      (if (CallFunction|IsWellKnownNameSupported)
        (CallFunction|BindorUnbindtoIslandQuerier true)))
    (:then_2
      (bind _outvolumemanager (Volume|GetVolumeManager)
        (:Valid
          (EventDispatchers|BindEventtoOnUserGeneratedContentRestrictionChanged (VolumeManager|GetVolumeforActor _outvolumemanager _self) (EventDispatchers|CreateEvent :!event OnUserGeneratedContentRestrictionChanged))
          (CallFunction|OnUserGeneratedContentRestrictionChanged (Class|FortVolume|GetUserGeneratedContentRestricted (VolumeManager|GetVolumeforActor _outvolumemanager _self))))))))

(fn SetupMeshTargetToCheckVisibility
  (Variables|VisibilityCheck|SetMeshTargetToCheckHUDVisibility (Variables|Debug|GetStaticMeshComponent)))

(fn SetupWaitForReadyCheck
  (params (FortPlayer :type Fort%20Player%20Controller%20Athena%20Object%20Reference))
  (bind _outputdelegate (EventDispatchers|CreateEvent :!event ReadyCheckCompleted))
  (bind _fortplayer (Variables|GetFortPlayer))
  (Utilities|IsValid FortPlayer
    (:Is%20Valid
      (Variables|Default|SetReadyCheckInteractionPlayer FortPlayer)
      (EventDispatchers|UnbindEventfromInGameMatchmakingReadyCheckComplete _fortplayer _outputdelegate)
      (EventDispatchers|BindEventtoInGameMatchmakingReadyCheckComplete _fortplayer _outputdelegate))))

(fn StartMatchmakingTimer
  (Utilities|Time|ClearandInvalidateTimerbyHandle (Variables|Default|GetMatchmakingTimer))
  (bind _returnvalue (Utilities|Time|SetTimerbyEvent (EventDispatchers|CreateEvent :!event HandleMatchmakingThrottleTimerComplete) (Class|FortMatchmakingV2|GetMatchmakingAttemptThrottleTime)))
  (Variables|Default|SetMatchmakingTimer _returnvalue)
  (Variables|Default|SetMatchmakingOnCooldown true))

(fn StartHUDVisibilityBlockingCheck
  (if (not (Utilities|Time|IsValidTimerHandle (Variables|VisibilityCheck|GetHUDVisibilityCheckTimer)))
    (bind _returnvalue (Utilities|Time|SetTimerbyEvent (EventDispatchers|CreateEvent :!event UpdateHUDVisibilityBlockCheck) 1.0 true))
    (Variables|VisibilityCheck|SetHUDVisibilityCheckTimer _returnvalue)))

(fn StoreClientPortalComponents
  (for _array_element (Fortnite|GetLocalFortPlayerControllers true)
    (Utilities|IsValid (Actor|GetComponentbyClass _array_element "/Script/FortniteGame.FortControllerComponent_Portal")
      (:Is%20Valid
        (Utilities|Map|Add (Variables|Default|GetClient%20Portal%20Components) _array_element (Actor|GetComponentbyClass _array_element "/Script/FortniteGame.FortControllerComponent_Portal"))))))

(fn UpdateCodeEntryScreen
  (params (Force%20Screen%20Update :type Boolean))
  (LocalVariables|SetMID (Material|GetMID (Variables|Default|GetCode%20Entry%20Screen)))
  (if (and (Math|Conversions|ToBoolean(Integer) (Utilities|Array|Length (Variables|CodeOverride|GetLocal%20Pawns%20in%20Range))) (Class|CreativeEnabledComponent|IsEnabled (Variables|Default|GetEnabledComponent)))
    ;[1 1 1] At Max this will only ever look twice. It's a check to ensure things work for both players in Splitscreen.
    (bind _array_element (Utilities|Array|ForEachLoopwithBreak (Variables|CodeOverride|GetLocal%20Pawns%20in%20Range))
      (:LoopBody
        (Utilities|IsLookingAtScreen _array_element
          (:True
            (LocalVariables|SetLookingAtScreen true))))
      (:Completed
        _))
    (else
      _))
  (if (or Force%20Screen%20Update (Math|Boolean|NotEqual(Boolean) (LocalVariables|GetLookingAtScreen) (Variables|CodeOverride|GetLocalPlayerLookingAtScreen)))
    (Rendering|Material|SetVectorParameterValue (Material|GetMID (Variables|Default|GetCode%20Entry%20Screen)) "Screen Tone" (select (Class|CreativeEnabledComponent|IsEnabled (Variables|Default|GetEnabledComponent)) (select (Utilities|Enum|Equal(Enum) (Variables|/Script/FortniteGame.FortPortalComponent|GetLinkCodeLockStatus (Variables|Default|GetFortPortal)) "Locked") (Variables|CodeOverride|GetScreenColorLocked) (select (LocalVariables|GetLookingAtScreen) (Variables|CodeOverride|GetScreenColorCanUpdate) (Variables|CodeOverride|GetScreenColorNotHighlighted))) (Variables|EnabledState|GetScreenColorDisabled)) :!class /Script/Engine.MaterialInstanceDynamic)
    (Variables|CodeOverride|SetLocalPlayerLookingAtScreen (LocalVariables|GetLookingAtScreen))
    (bind _hud_widget_component (Variables|Default|GetHUD%20Widget%20Component)
      (:then
        (Class|CreativeHeaderDescriptionHUDComponent|SpawnOrUpdateHUD _hud_widget_component "NewEnumerator1"))
      (:else))))

(fn UpdateEnabled
  (params (Enabled :type Boolean))
  (Networking|CanExecuteCosmeticEvents
    (:True
      (Collision|SetCollisionEnabled (Variables|Default|GetHUD%20Visibility%20Volume) (select Enabled "QueryOnly" "NoCollision")) ; Disabling or enabling collision on the visibility volume\nwill trigger the overlap events to enable/disable the HUD.
      (CallFunction|UpdatePortalVisuals)
      (EnabledState|UpdatePoweredVisuals)
      (if (not (Variables|GetEnabled))
        (bind _hud_widget_component (Variables|Default|GetHUD%20Widget%20Component)
          (:then
            (Class|CreativeHeaderDescriptionHUDComponent|ForceShowWidget _hud_widget_component false))
          (:else))))))

(fn UpdateHUD
  (params (Matchmaking%20State :type ECreativeMatchmakingPortalState%20Enum :default "NewEnumerator0"))
  (Variables|IslandDetails|SetCurrent%20Matchmaking%20State Matchmaking%20State)
  (Networking|CanExecuteCosmeticEvents
    (:True
      (bind _hud_widget_component (Variables|Default|GetHUD%20Widget%20Component)
        (:then
          (Class|CreativeHeaderDescriptionHUDComponent|SpawnOrUpdateHUD _hud_widget_component "NewEnumerator1"))
        (:else)))
    (:False
      (bind _hud_visibility_volume (Variables|Default|GetHUD%20Visibility%20Volume)
        (:then
          (Components|DestroyComponent _hud_visibility_volume) ; We don't have any reason to show a widget, so no need to have a visibility checking volume.\nNo need to destroy the widget also, it will do that by itself.
        )
        (:else)))))

(fn UpdateIslandDetails
  (params (Code :type String) (Logic%20on%20Client :type Boolean) (Client%20Player%20Controller :type Fort%20Player%20Controller%20Object%20Reference))
  (bind _client_portal_components (Variables|Default|GetClient%20Portal%20Components))
  (bind _portal_loop_audio (Variables|Default|GetPortal%20Loop%20Audio))
  (Utilities|BasicCodeValidation Code
    (:Valid%20Code
      (Variables|Default|SetCurrent%20Island%20Code (Variables|GetCode))
      (CallFunction|PlayPortalAudio)
      (Networking|SwitchHasAuthority
        (:Authority
          (if (not (Variables|GetLogic%20on%20Client))
            (bind _returnvalue (CallFunction|CreateLinkId (Variables|GetCode)))
            (Class|FortPortalComponent|ServerSetLinkCode (Variables|Default|GetFortPortal) _returnvalue (Variables|DAD|GetbDADRequestedUpdate) (Variables|DAD|GetbIslandCodeFromDAD) false)))
        (:Remote
          (CallFunction|IslandDataNoSuccess "NewEnumerator1")
          (if (Variables|GetLogic%20on%20Client)
            (bind _client_player_controller (Variables|GetClient%20Player%20Controller)
              (:then
                (if (.ReturnValue (Utilities|Map|Find _client_portal_components _client_player_controller))
                  (bind _returnvalue_2 (CallFunction|CreateLinkId (Variables|GetCode)))
                  (Class|FortControllerComponentPortal|ServerSetPortalLinkCode (Utilities|Map|Find _client_portal_components _client_player_controller) (Variables|Default|GetFortPortal) _returnvalue_2)))
              (:else))))))
    (:Invalid%20Code
      (Variables|Default|SetCurrent%20Island%20Code "")
      (CallFunction|IslandDataNoSuccess "NewEnumerator2"))
    (:CurrentCode
      (CallFunction|UpdateHUD (Variables|IslandDetails|GetCurrent%20Matchmaking%20State))
      (CallFunction|PlayPortalAudio))
    (:DefaultCode
      (Variables|Default|SetCurrent%20Island%20Code "")
      (CallFunction|IslandDataNoSuccess "NewEnumerator4"))) ; Runs Validation Logic prior to submission \nagainst known bad options (defualt, string length too short).\nTriggers appropriate error messages.
  (if (Audio|Components|Audio|IsPlaying _portal_loop_audio)
    (Audio|Components|Audio|FadeOut _portal_loop_audio 0.5 0.0 "Linear" :!class /Script/Engine.AudioComponent)))

(fn UpdateLockMode
  (bind _returnvalue (CallFunction|ConsoleIsVisible))
  (bind _returnvalue_2 (not _returnvalue))
  (Variables|/Script/FortniteGame.FortPortalComponent|SetLinkCodeLockMode (select (and (> (Variables|Default|GetCodeOverrideCooldown) 0) (not (Utilities|Enum|Equal(Enum) (Variables|UserOptions|GetCanOverrideCode) "AlwaysLocked"))) "WindowLocked" (Variables|UserOptions|GetCanOverrideCode)) (Variables|Default|GetFortPortal))
  (Collision|SetCollisionEnabled (Variables|Default|GetCode%20Entry%20Screen) (select _returnvalue_2 "NoCollision" "QueryAndPhysics"))
  (CallFunction|UpdateWindowLockCooldown)
  (Networking|CanExecuteCosmeticEvents
    (:True
      (Development|SetHiddeninGame (Variables|Default|GetCode%20Entry%20Screen) _returnvalue_2 false))))

(fn UpdatePortalVisuals
  (Development|SetHiddeninGame (Variables|Default|GetThumbnailPlane) (not (CallFunction|PortalThumbnailIsVisible)) false))

(fn UpdatePoweredVisuals
  (bind _enabledcomponent (Variables|Default|GetEnabledComponent))
  (bind _enabled (Class|CreativeEnabledComponent|IsEnabled _enabledcomponent))
  (Rendering|Material|SetScalarParameterValue (Material|GetMID (Variables|Default|GetCode%20Entry%20Screen)) "UseTexture" (select _enabled 1.0 0.0) :!class /Script/Engine.MaterialInstanceDynamic)
  (Rendering|Material|SetVectorParameterValue (Material|GetMIDfromMaterial (Variables|Debug|GetStaticMeshComponent) 1) "Emissive" (select _enabled (Variables|EnabledState|GetEmissive%20Color%20Enabled) (Variables|EnabledState|GetScreenColorDisabled)) :!class /Script/Engine.MaterialInstanceDynamic)
  (Development|SetHiddeninGame (Variables|Default|GetLight) (not _enabled) false)
  (if (Class|CreativeEnabledComponent|IsEnabled (Variables|Default|GetEnabledComponent))
    (CodeOverride|UpdateCodeEntryScreen true)
    (else
      (Rendering|Material|SetVectorParameterValue (Material|GetMID (Variables|Default|GetCode%20Entry%20Screen)) "Screen Tone" (Variables|EnabledState|GetScreenColorDisabled) :!class /Script/Engine.MaterialInstanceDynamic))))

(fn UpdateThumbnailDetails
  (params (Success :type Boolean) (Texture :type Texture%202DDynamic%20Object%20Reference))
  (bind _thumbnail_mid (Variables|Default|GetThumbnail%20MID))
  (if Success
    (Utilities|IsValid Texture
      (:Is%20Valid
        (Variables|IslandDetails|SetThumbnailTexture Texture)
        (Networking|CanExecuteCosmeticEvents
          (:True
            (if (not (Utilities|Equal(Object) (Rendering|Material|GetMaterial (Variables|Default|GetThumbnailPlane) 0) (Variables|Default|GetThumbnail%20MID)))
              (Rendering|Material|SetMaterial (Variables|Default|GetThumbnailPlane) 0 (Variables|Default|GetThumbnail%20MID)))
            (Rendering|Material|SetTextureParameterValue _thumbnail_mid "Texture" (Variables|IslandDetails|GetThumbnailTexture))
            (Rendering|Material|SetVectorParameterValue _thumbnail_mid "Color" (Variables|Default|GetActive%20Portal%20Color) :!class /Script/Engine.MaterialInstanceDynamic)
            (CallFunction|UpdatePortalVisuals))))
      (:Is%20Not%20Valid
        _))
    (else
      _))
  (CallFunction|UpdatePortalVisuals))

(fn UpdateLight
  ;[1 0.5253 0 0.2] Sets the light intensity from a percentage input to a defined output.
  (Networking|CanExecuteCosmeticEvents
    (:True
      (Rendering|Components|Light|SetIntensity (Variables|Default|GetLight) (Math|Float|MapRangeUnclamped (Variables|UserOptions|GetLightStrengthPercent) 0.0 1e+02 0.0 (Variables|Default|GetMaxLightIntensity))))))

(fn UpdateWindowLockCooldown
  (Networking|SwitchHasAuthority
    (:Authority
      (if (Utilities|Enum|Equal(Enum) (Variables|/Script/FortniteGame.FortPortalComponent|GetLinkCodeLockMode (Variables|Default|GetFortPortal)) "WindowLocked")
        (Class|FortPortalComponent|SetLockOverrideExpirationSeconds (Variables|Default|GetFortPortal) (Variables|Default|GetCodeOverrideCooldown))
        (else
          (Class|FortPortalComponent|SetLockExpirationSecondstoNotOverrideCVar (Variables|Default|GetFortPortal)))))))

(fn ConstructionScript
  (|CallParentFunction)
  (Networking|SwitchHasAuthority
    (:Authority
      (Collision|SetCollisionEnabled (Variables|Default|GetHUD%20Visibility%20Volume) "NoCollision"))
    (:Remote
      ;[1 0.6282 0 0.5] Creating MID here as the phone tool reverts back to its initial material when pasting the device (and so creating it in construction it will return to the dynamic Material)
      (bind _returnvalue (Rendering|Material|CreateDynamicMaterialInstance (Variables|Default|GetThumbnailPlane) 0 0 "None" :!class /Script/Engine.PrimitiveComponent))
      (Variables|Default|SetThumbnail%20MID _returnvalue))))
