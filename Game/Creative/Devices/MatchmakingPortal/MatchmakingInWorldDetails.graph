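# Widget for the Creative Matchmaking Portal's in-world details panel, showing a header or error message and a matchmaking status line.
# On Construct it collapses itself if IsOnPortalHudEnabled returns true.
# SetHeaderDetails uppercases text except island codes, caches it, and switches colors, alert icon and border between normal and error styles.
# Other functions toggle visibility, hiding empty headers, and switch the status between "Matchmaking Enabled" and "Private Island".
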WidgetBlueprint /Game/Creative/Devices/MatchmakingPortal/MatchmakingInWorldDetails
graphs [
 EventGraph {nodes [
   Event Construct()
   > SetVisibility(InVisibility=Select(Option 0=Visible, Option 1=Collapsed, Index=IsOnPortalHudEnabled!() -> (ReturnValue:bool)))
  ]}
 SetHeaderDetails {nodes [
   n1 = SetHeaderDetails!() -> (then, Header:Text, IsNotError:bool=~) {is=title K2Node_FunctionEntry}
   n2 = "To String (Text)"(InText=@n1)
   > n3 = Set CachedHeaderText(CachedHeaderText:Text="To Text (String)"(InString=ToUpperExceptIslandCode!(InString:String=@n2) -> (ReturnValue:String))) -> (Output_Get:Text)
   > "SetText (Text)"(self=Get HeaderText() -> (HeaderText:Common Text Object Reference), InText=@n3)
   > SetHeaderVisible!(execute, self, bVisible:bool=Get bHeaderDesiredVisible() -> (bHeaderDesiredVisible:bool)) -> (then)
   > SetColorAndOpacity(
    self=Get HeaderText()
    InColorAndOpacity=Select(
     Option 0:Slate Color Structure=Get ErrorTextColor() -> (ErrorTextColor:Slate Color Structure)
     Option 1:Slate Color Structure=Get NormalTextColor() -> (NormalTextColor:Slate Color Structure)
     Index=Get IsNotError() -> (IsNotError:bool)
    ) -> (ReturnValue:Slate Color Structure)
   )
   > SetBrushColor(
    self=Get Border_Header() -> (Border_Header:Border Object Reference)
    InBrushColor=SelectColor(
     A=Get NormalBorderColor() -> (NormalBorderColor:Linear Color Structure)
     B=Get ErrorBorderColor() -> (ErrorBorderColor:Linear Color Structure)
     bPickA=Get IsNotError()
    )
   )
   > SetShadowColorAndOpacity!(
    execute
    self:Text Object Reference=Get HeaderText()
    InShadowColorAndOpacity:Linear Color Structure=SelectColor(
     A=Get NormalShadowColor() -> (NormalShadowColor:Linear Color Structure)
     B=Get ErrorShadowColor() -> (ErrorShadowColor:Linear Color Structure)
     bPickA=Get IsNotError()
    )
   ) -> (then)
   > Set Outline Color!(
    In:Exec
    CommonTextWidget:Common Text Object Reference=Get HeaderText()
    Color:Linear Color Structure=SelectColor(
     A=Get NormalOutlineColor() -> (NormalOutlineColor:Linear Color Structure)
     B=Get ErrorOutlineColor() -> (ErrorOutlineColor:Linear Color Structure)
     bPickA=Get IsNotError()
    )
   ) -> (Out:Exec) {is=title K2Node_MacroInstance}
   > SetVisibility(
    self=[Get Image_96() -> (Image_96:Image Object Reference), Get Border_Header()]
    InVisibility=Select(Option 0:ESlateVisibility Enum=SelfHitTestInvisible, Option 1:ESlateVisibility Enum=Collapsed, Index=Get IsNotError()) -> (ReturnValue:ESlateVisibility Enum)
   )
  ]}
 SetHeaderVisible {nodes [
   n1 = SetHeaderVisible!() -> (then, bVisible:bool=~) {is=title K2Node_FunctionEntry}
   > n2 = Set bHeaderDesiredVisible(bHeaderDesiredVisible:bool=@n1) -> (Output_Get:bool)
   n3 = NOT Boolean(A=TextIsEmpty(InText=Get CachedHeaderText() -> (CachedHeaderText:Text)))
   > SetVisibility(
    self=[Get HeaderText(), Get Border_Header()]
    InVisibility=Select(Option 0=Collapsed, Option 1=SelfHitTestInvisible, Index=AND Boolean(A=@n2, B=@n3))
   )
  ]}
 SetMatchmakingEnabledText {nodes [
   n1 = SetMatchmakingEnabledText!() -> (then, PrivateIsland:bool=~) {is=title K2Node_FunctionEntry}
   > "SetText (Text)"(
    self=Get Text_MatchmakingEnabled() -> (Text_MatchmakingEnabled:Common Text Object Reference)
    InText=Select(
     Option 0:Text=Get MatchmakingEnabledText() -> (MatchmakingEnabledText:Text)
     Option 1:Text=Get PrivateIslandText() -> (PrivateIslandText:Text)
     Index=@n1
    ) -> (ReturnValue:Text)
   )
   > Return Node()
  ]}
 SetMatchmakingEnabledVisible {nodes [
   n1 = SetMatchmakingEnabledVisible!() -> (then, bVisible:bool=~) {is=title K2Node_FunctionEntry}
   > SetVisibility(
    self=Get Text_MatchmakingEnabled()
    InVisibility=Select(Option 0:ESlateVisibility Enum=Collapsed, Option 1:ESlateVisibility Enum=SelfHitTestInvisible, Index=@n1) -> (ReturnValue:ESlateVisibility Enum)
   )
  ]}
]
variables [
 bHeaderDesiredVisible {type=boolean}
 CachedHeaderText {type=string, default=""}
 ErrorBorderColor {type=LinearColor, default=<rgba 0.0048 0.1946 0.3712 1>}
 ErrorOutlineColor {type=LinearColor, default=<rgba 0 0.1882 0.251 1>}
 ErrorShadowColor {type=LinearColor, default=<rgba 0 0 0 1>}
 ErrorTextColor {type=SlateColor, default {colorUseRule=UseColor_Specified, specifiedColor=<rgba 1 1 1 1>}}
 MatchmakingEnabledText {type=string, default=Matchmaking Enabled}
 NormalBorderColor {type=LinearColor, default=<rgba 0 0 0 0>}
 NormalOutlineColor {type=LinearColor, default=<rgba 0 0 0 0>}
 NormalShadowColor {type=LinearColor, default=<rgba 0 0 0 0>}
 NormalTextColor {type=SlateColor, default {colorUseRule=UseColor_Specified, specifiedColor=<rgba 1 1 1 1>}}
 PrivateIslandText {type=string, default=Private Island}
]
widgets {
 info {inheritedWidgetCount=0, parentClass=&UserWidget, rootWidgetClass=&Overlay, widgetCount=12}
 widgets [
  Overlay_2 {widgetClassPath=&Overlay, widget=Overlay_2}
  SizeBox_53 {
   widgetClassPath=&SizeBox
   parent=Overlay_2
   widget=SizeBox_53
   properties {
    bOverride_HeightOverride=true
    bOverride_WidthOverride=true
    heightOverride=94
    maxDesiredHeight=50
    maxDesiredWidth=700
    minDesiredHeight=50
    minDesiredWidth=374
    widthOverride=373
   }
   slot {class=OverlaySlot}
  }
  Overlay_167 {widgetClassPath=&Overlay, parent=SizeBox_53, widget=Overlay_167, slot {class=SizeBoxSlot}}
  CommonBorder_106 {
   widgetClassPath=&CommonBorder
   parent=Overlay_167
   widget=CommonBorder_106
   properties {brushColor=<rgba 0 0.0062 0.0677 1>, padding=<ltrb 0 0 0 0>}
   slot {class=OverlaySlot, properties {horizontalAlignment=HAlign_Fill, verticalAlignment=VAlign_Fill}}
  }
  VerticalBox_0 {
   widgetClassPath=&VerticalBox
   parent=Overlay_167
   widget=VerticalBox_0
   slot {class=OverlaySlot, properties {horizontalAlignment=HAlign_Fill, verticalAlignment=VAlign_Fill}}
  }
  Spacer_59 {
   widgetClassPath=&Spacer
   parent=VerticalBox_0
   widget=Spacer_59
   slot {class=VerticalBoxSlot, properties {verticalAlignment=VAlign_Top, padding=<ltrb 10 10 10 0>}}
  }
  Overlay_0 {widgetClassPath=&Overlay, parent=VerticalBox_0, widget=Overlay_0, slot {class=VerticalBoxSlot}}
  Border_Header {
   widgetClassPath=&Border
   bIsVariable=true
   parent=Overlay_0
   widget=Border_Header
   properties {brushColor=<rgba 0 0 0 0>}
   slot {class=OverlaySlot, properties {horizontalAlignment=HAlign_Fill, verticalAlignment=VAlign_Fill, padding=<ltrb 5 0 5 5>}}
  }
  HorizontalBox_0 {
   widgetClassPath=&HorizontalBox
   parent=Overlay_0
   widget=HorizontalBox_0
   slot {class=OverlaySlot, properties {horizontalAlignment=HAlign_Fill, verticalAlignment=VAlign_Fill, padding=<ltrb 10 0 10 0>}}
  }
  Image_96 {
   widgetClassPath=&Image
   bIsVariable=true
   parent=HorizontalBox_0
   widget=Image_96
   properties {brush {
     bIsDynamicallyLoaded=false
     drawAs=Image
     imageType=NoImage
     mirroring=NoMirror
     resourceName=None
     resourceObject=/Game/UI/Foundation/Textures/Icons/HUD/T-Icon-Alert2-32
     tiling=NoTile
     imageSize=<32 32>
     margin=<ltrb 0 0 0 0>
     outlineSettings {
      bUseBrushTransparency=false
      roundingType=HalfHeightRadius
      width=0
      color {colorUseRule=UseColor_Specified, specifiedColor=<rgba 0 0 0 0>}
      cornerRadii=<0 0 0 0>
     }
     tintColor {colorUseRule=UseColor_Specified, specifiedColor=<rgba 1 1 1 1>}
     uVRegion {bIsValid=false, max=<0 0>, min=<0 0>}
    }}
   slot {class=HorizontalBoxSlot, properties {horizontalAlignment=HAlign_Center, verticalAlignment=VAlign_Top, padding=<ltrb 0 2 0 0>}}
  }
  HeaderText {
   widgetClassPath=&CommonTextBlock
   bIsVariable=true
   parent=HorizontalBox_0
   widget=HeaderText
   properties {
    autoWrapText=true
    text="Unmoderated Island Codes Require A Version Tag\r\ne.g. 1234-5678-9123?v=4"
    wrappingPolicy=AllowPerCharacterWrapping
    font {
     bForceMonospaced=false
     bMaterialIsStencil=false
     fontMaterial=null
     fontObject=&BurbankSmall
     letterSpacing=0
     monospacedWidth=1
     size=12
     skewAmount=0
     typefaceFontName=Medium
     outlineSettings {
      bApplyOutlineToDropShadows=false
      bMiteredCorners=false
      bSeparateFillAlpha=false
      outlineMaterial=null
      outlineSize=1
      outlineColor=<rgba 0 0 0 0>
     }
    }
    shadowOffset=<2 2>
   }
   slot {class=HorizontalBoxSlot, properties {padding=<ltrb 0 2 4 2>, size {sizeRule=Fill, value=1}}}
  }
  Text_MatchmakingEnabled {
   widgetClassPath=&CommonTextBlock
   bIsVariable=true
   parent=VerticalBox_0
   widget=Text_MatchmakingEnabled
   properties {
    text=Matchmaking Enabled
    colorAndOpacity {colorUseRule=UseColor_Specified, specifiedColor=<rgba 1 0.5253 0 1>}
    font {
     bForceMonospaced=false
     bMaterialIsStencil=false
     fontMaterial=null
     fontObject=&BurbankSmall
     letterSpacing=0
     monospacedWidth=1
     size=10
     skewAmount=0
     typefaceFontName=Medium
     outlineSettings {
      bApplyOutlineToDropShadows=false
      bMiteredCorners=false
      bSeparateFillAlpha=false
      outlineMaterial=null
      outlineSize=0
      outlineColor=<rgba 0 0.066 0.066 0.3>
     }
    }
   }
   slot {class=VerticalBoxSlot, properties {verticalAlignment=VAlign_Top, padding=<ltrb 10 0 10 0>}}
  }
 ]
}
