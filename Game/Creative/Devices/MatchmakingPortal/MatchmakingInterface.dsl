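; MatchmakingInterface declares the function contract for the Creative matchmaking portal device, apparently as a Blueprint interface.
; It covers interaction details, whether the portal has a second interaction for a given player pawn, and sending a Creative island link from a requesting player controller.
; All three functions only return false, so the file itself holds no behavior.
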(fn AthenaInteractDetails
  (return false false))

(fn MatchmakingPortalHasSecondInteraction
  (params (InteractingPawn :type Fort%20Player%20Pawn%20Object%20Reference))
  (return false))

(fn SendIslandLink
  (params (Requesting%20Player%20Controller :type Fort%20Player%20Controller%20Object%20Reference) (Link :type Fort%20Creative%20Island%20Link%20Object%20Reference))
  (return false))
