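# Creative matchmaking portal HUD widget that shows the destination island's title, creator and player count, with a keybind prompt to cancel matchmaking.
# SetupIslandInfo fills the text from a FortGameActivity, then calls Open; OnHide removes the widget after a short delay.
# The parent FortCreativeMatchmakingHUD's events show the prompt on party leader setup and collapse it when matchmaking completes.
# Completing the keybind cancels an in-progress party ready check, otherwise matchmaking itself.
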WidgetBlueprint /Game/Creative/Devices/MatchmakingPortal/MatchmakingPortalHUD
graphs [
 EventGraph {nodes [
   OnHide() {is=title K2Node_CustomEvent}
   > Delay(Duration=0.5)
   > RemoveFromParent()
   Event BP_PartyLeaderSetup() {is=title K2Node_Event}
   > SetVisibility(self=Get Border_Cancel() -> (Border_Cancel:Border Object Reference), InVisibility=Visible)
   n1 = Event BP_OnKeybindComplete!() -> (OutputDelegate, then, FortPC:Fort Player Controller Athena Object Reference) {is=title K2Node_Event}
   > n2 = Branch(Condition=IsPartyMemberReadyCheckStatusInProgress!(self:Fort Player Controller Athena Object Reference=@n1.FortPC) -> (ReturnValue:bool))
   > CancelCreativeMatchmakingReadyCheck(self:Fort Player Controller Athena Object Reference=@n1.FortPC)
   CancelMatchmaking!(
    execute=@n2.else
    self:Fort Player Controller Athena Object Reference=@n1.FortPC
    CancelReason:EMatchmakingCancelReasonV2 Enum=CreativeCanceledByLeader
   ) -> (then)
   Event BP_OnMatchmakingComplete!() -> (OutputDelegate, then, bSuccess:bool) {is=title K2Node_Event}
   > SetVisibility(self=Get Border_Cancel(), InVisibility=Collapsed)
  ]}
 SetupIslandInfo {nodes [
   n1 = SetupIslandInfo!() -> (then, IslandData:Fort Game Activity Object Reference) {is=title K2Node_FunctionEntry}
   > "SetText (Text)"(
    self=Get Text_Creator() -> (Text_Creator:Common Text Object Reference)
    InText="To Text (String)"(InString=Get Creator!(self:Fort Game Activity Object Reference=@n1) -> (Creator:String))
   )
   n2 = GetMatchmakingSettingsV1!(self:Fort Game Activity Object Reference=@n1) -> (ReturnValue:Creative Island Matchmaking Settings Structure)
   > "SetText (Text)"(
    self=Get Text_PlayerCount() -> (Text_PlayerCount:Common Text Object Reference)
    InText="To Text (Integer)"(Value=Break Creative Island Matchmaking Settings!(CreativeIslandMatchmakingSettings:"Creative Island Matchmaking Settings Structure (by ref)"=@n2) -> (PlayerCount:int, SelectedJoinInProgressType:int) {is=title K2Node_BreakStruct})
   )
   > "SetText (Text)"(
    self=Get Text_IslandName() -> (Text_IslandName:Common Text Object Reference)
    InText="To Text (String)"(InString=Get Title!(self:Fort Game Activity Object Reference=@n1) -> (Title:String))
   )
   > Open(self:Fort Creative Matchmaking HUD Object Reference)
  ]}
]
variables [
 RequiredPressTime {type=number, default=2}
 TimerHandle {type=TimerHandle}
]
widgets {
 info {inheritedWidgetCount=1, parentClass=/Script/FortniteUI.FortCreativeMatchmakingHUD, rootWidgetClass=&Overlay, widgetCount=23}
 widgets [
  Overlay_13 {widgetClassPath=&Overlay, widget=Overlay_13}
  VerticalBox_31 {widgetClassPath=&VerticalBox, parent=Overlay_13, widget=VerticalBox_31, slot {class=OverlaySlot}}
  Overlay_50 {widgetClassPath=&Overlay, parent=VerticalBox_31, widget=Overlay_50, slot {class=VerticalBoxSlot}}
  SizeBox_IslandInfo {
   widgetClassPath=&SizeBox
   bIsVariable=true
   parent=Overlay_50
   widget=SizeBox_IslandInfo
   properties {bOverride_MaxDesiredWidth=true, heightOverride=400, maxDesiredHeight=540, maxDesiredWidth=580, widthOverride=960}
   slot {class=OverlaySlot, properties {horizontalAlignment=HAlign_Right, verticalAlignment=VAlign_Bottom}}
  }
  Overlay_54 {widgetClassPath=&Overlay, parent=SizeBox_IslandInfo, widget=Overlay_54, slot {class=SizeBoxSlot}}
  VerticalBox_70 {
   widgetClassPath=&VerticalBox
   parent=Overlay_54
   widget=VerticalBox_70
   slot {class=OverlaySlot, properties {padding=<ltrb 10 10 10 10>}}
  }
  Text_IslandName {
   widgetClassPath=&CommonTextBlock
   bIsVariable=true
   parent=VerticalBox_70
   widget=Text_IslandName
   properties {
    autoWrapText=true
    text=EXAMPLE TITLE
    font {
     bForceMonospaced=false
     bMaterialIsStencil=false
     fontMaterial=null
     fontObject=&BurbankSmall
     letterSpacing=0
     monospacedWidth=1
     size=36
     skewAmount=0
     typefaceFontName=Bold
     outlineSettings {
      bApplyOutlineToDropShadows=false
      bMiteredCorners=false
      bSeparateFillAlpha=false
      outlineMaterial=null
      outlineSize=0
      outlineColor=<rgba 0 0 0 1>
     }
    }
    shadowColorAndOpacity=<rgba 0 0 0 0.5>
    shadowOffset=<2 2>
   }
   slot {class=VerticalBoxSlot}
  }
  HorizontalBox_Creator {
   widgetClassPath=&HorizontalBox
   bIsVariable=true
   parent=VerticalBox_70
   widget=HorizontalBox_Creator
   slot {class=VerticalBoxSlot}
  }
  Text_CreatorLabel {
   widgetClassPath=&CommonTextBlock
   parent=HorizontalBox_Creator
   widget=Text_CreatorLabel
   properties {
    text=CREATED BY:
    textTransformPolicy=ToUpper
    colorAndOpacity {colorUseRule=UseColor_Specified, specifiedColor=<rgba 1 1 1 0.75>}
    font {
     bForceMonospaced=false
     bMaterialIsStencil=false
     fontMaterial=null
     fontObject=&Burbank
     letterSpacing=0
     monospacedWidth=1
     size=24
     skewAmount=0
     typefaceFontName=BurbankRegular-Black
     outlineSettings {
      bApplyOutlineToDropShadows=false
      bMiteredCorners=false
      bSeparateFillAlpha=false
      outlineMaterial=null
      outlineSize=0
      outlineColor=<rgba 0 0 0 1>
     }
    }
    shadowColorAndOpacity=<rgba 0 0 0 0.5>
    shadowOffset=<2 2>
   }
   slot {class=HorizontalBoxSlot, properties {verticalAlignment=VAlign_Center, padding=<ltrb 0 0 8 0>}}
  }
  Text_Creator {
   widgetClassPath=&CommonTextBlock
   bIsVariable=true
   parent=HorizontalBox_Creator
   widget=Text_Creator
   properties {
    autoWrapText=true
    text="[CREATOR]"
    textTransformPolicy=ToUpper
    colorAndOpacity {colorUseRule=UseColor_Specified, specifiedColor=<rgba 0 1 1 1>}
    font {
     bForceMonospaced=false
     bMaterialIsStencil=false
     fontMaterial=null
     fontObject=&Burbank
     letterSpacing=0
     monospacedWidth=1
     size=24
     skewAmount=0
     typefaceFontName=BurbankRegular-Black
     outlineSettings {
      bApplyOutlineToDropShadows=false
      bMiteredCorners=false
      bSeparateFillAlpha=false
      outlineMaterial=null
      outlineSize=0
      outlineColor=<rgba 0 0 0 1>
     }
    }
    shadowColorAndOpacity=<rgba 0 0 0 0.5>
    shadowOffset=<2 2>
   }
   slot {class=HorizontalBoxSlot, properties {verticalAlignment=VAlign_Center, size {sizeRule=Fill, value=1}}}
  }
  Spacer_0 {widgetClassPath=&Spacer, parent=HorizontalBox_Creator, widget=Spacer_0, properties {size=<5 1>}, slot {class=HorizontalBoxSlot}}
  Overlay_0 {
   widgetClassPath=&Overlay
   parent=HorizontalBox_Creator
   widget=Overlay_0
   slot {class=HorizontalBoxSlot, properties {verticalAlignment=VAlign_Center, padding=<ltrb 10 0 0 0>}}
  }
  Image_PlayerCount_Shadow {
   widgetClassPath=&Image
   bIsVariable=true
   parent=Overlay_0
   widget=Image_PlayerCount_Shadow
   properties {
    brush {
     bIsDynamicallyLoaded=false
     drawAs=Image
     imageType=NoImage
     mirroring=NoMirror
     resourceName=None
     resourceObject=/Game/UI/Foundation/Textures/Icons/Social/T_Icon_Matchmaking_Squad_128
     tiling=NoTile
     imageSize=<32 32>
     margin=<ltrb 0 0 0 0>
     outlineSettings {
      bUseBrushTransparency=false
      roundingType=HalfHeightRadius
      width=0
      color {colorUseRule=UseColor_Specified, specifiedColor=<rgba 0 0 0 0>}
      cornerRadii=<0 0 0 0>
     }
     tintColor {colorUseRule=UseColor_Specified, specifiedColor=<rgba 1 1 1 1>}
     uVRegion {bIsValid=false, max=<0 0>, min=<0 0>}
    }
    colorAndOpacity=<rgba 0 0 0 0.5>
    renderTransform {angle=0, scale=<1 1>, shear=<0 0>, translation=<3 3>}
   }
   slot {class=OverlaySlot, properties {padding=<ltrb 0 0 3 0>}}
  }
  Image_PlayerCount {
   widgetClassPath=&Image
   bIsVariable=true
   parent=Overlay_0
   widget=Image_PlayerCount
   properties {brush {
     bIsDynamicallyLoaded=false
     drawAs=Image
     imageType=NoImage
     mirroring=NoMirror
     resourceName=None
     resourceObject=&T_Icon_Matchmaking_Squad_128
     tiling=NoTile
     imageSize=<32 32>
     margin=<ltrb 0 0 0 0>
     outlineSettings {
      bUseBrushTransparency=false
      roundingType=HalfHeightRadius
      width=0
      color {colorUseRule=UseColor_Specified, specifiedColor=<rgba 0 0 0 0>}
      cornerRadii=<0 0 0 0>
     }
     tintColor {colorUseRule=UseColor_Specified, specifiedColor=<rgba 1 1 1 1>}
     uVRegion {bIsValid=false, max=<0 0>, min=<0 0>}
    }}
   slot {class=OverlaySlot, properties {padding=<ltrb 0 0 3 0>}}
  }
  Text_PlayerCount {
   widgetClassPath=&CommonTextBlock
   bIsVariable=true
   parent=HorizontalBox_Creator
   widget=Text_PlayerCount
   properties {
    text=12
    textTransformPolicy=ToUpper
    font {
     bForceMonospaced=false
     bMaterialIsStencil=false
     fontMaterial=null
     fontObject=&Burbank
     letterSpacing=0
     monospacedWidth=1
     size=24
     skewAmount=0
     typefaceFontName=BurbankRegular-Black
     outlineSettings {
      bApplyOutlineToDropShadows=false
      bMiteredCorners=false
      bSeparateFillAlpha=false
      outlineMaterial=null
      outlineSize=0
      outlineColor=<rgba 0 0 0 1>
     }
    }
    shadowColorAndOpacity=<rgba 0 0 0 0.5>
    shadowOffset=<2 2>
   }
   slot {class=HorizontalBoxSlot, properties {verticalAlignment=VAlign_Center, padding=<ltrb 0 0 8 0>}}
  }
  Spacer_207 {widgetClassPath=&Spacer, parent=VerticalBox_70, widget=Spacer_207, properties {size=<1 100>}, slot {class=VerticalBoxSlot}}
  Spacer_54 {widgetClassPath=&Spacer, parent=VerticalBox_70, widget=Spacer_54, properties {size=<1 25>}, slot {class=VerticalBoxSlot}}
  Border_Cancel {
   widgetClassPath=&Border
   bIsVariable=true
   parent=VerticalBox_70
   widget=Border_Cancel
   properties {visibility=Collapsed, brushColor=<rgba 0 0 0 0.5>}
   slot {class=VerticalBoxSlot, properties {horizontalAlignment=HAlign_Center}}
  }
  HorizontalBox_0 {widgetClassPath=&HorizontalBox, parent=Border_Cancel, widget=HorizontalBox_0, slot {class=BorderSlot}}
  ScaleBox_0 {
   widgetClassPath=&ScaleBox
   parent=HorizontalBox_0
   widget=ScaleBox_0
   properties {stretch=ScaleToFitY}
   slot {class=HorizontalBoxSlot}
  }
  KeybindWidget_MMHUD {
   widgetClassPath=/FortUILibrary/Input/KeybindWidget
   bInherited=true
   bIsVariable=true
   parent=ScaleBox_0
   widget=KeybindWidget_MMHUD
   slot {class=ScaleBoxSlot}
  }
  Overlay_1 {
   widgetClassPath=&Overlay
   parent=HorizontalBox_0
   widget=Overlay_1
   slot {class=HorizontalBoxSlot, properties {verticalAlignment=VAlign_Center, padding=<ltrb 5 10 0 10>, size {sizeRule=Fill, value=1}}}
  }
  Text_Cancel {
   widgetClassPath=&CommonTextBlock
   parent=Overlay_1
   widget=Text_Cancel
   properties {
    text=Cancel Matchmaking
    font {
     bForceMonospaced=false
     bMaterialIsStencil=false
     fontMaterial=null
     fontObject=&BurbankSmall
     letterSpacing=0
     monospacedWidth=1
     size=20
     skewAmount=0
     typefaceFontName=Bold
     outlineSettings {
      bApplyOutlineToDropShadows=false
      bMiteredCorners=false
      bSeparateFillAlpha=false
      outlineMaterial=null
      outlineSize=0
      outlineColor=<rgba 0 0 0 1>
     }
    }
    shadowColorAndOpacity=<rgba 0 0 0 0.5>
    shadowOffset=<2 2>
   }
   slot {class=OverlaySlot}
  }
 ]
}
